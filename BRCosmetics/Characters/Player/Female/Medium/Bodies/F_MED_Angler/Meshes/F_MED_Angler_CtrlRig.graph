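# Control Rig for the F_MED_Angler female character body that adds secondary motion to its arm and leg fins.
# On each side, the local rotations of dyn_legFinTop_03 and dyn_armFin_01 drive the other fins, scaled by slerp from identity and smoothed by AccumulateLerp with varying blends for staggered lag.
# AimBone nodes point each dyn_armFin_03_Base at an offset from dyn_armFin_02.
# Several unconnected experimental and debug nodes remain.
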ControlRigBlueprint /BRCosmetics/Characters/Player/Female/Medium/Bodies/F_MED_Angler/Meshes/F_MED_Angler_CtrlRig
graphs [
 RigVMFunctionLibrary {links [], nodes []}
 RigVMModel {nodes [
   n1 = &RigUnit_BeginExecution()
   n2 = &RigUnit_GetBoneTransform(Bone=dyn_legFinTop_03_r, Space=LocalSpace)
   /Script/ControlRig.RigUnit_SetBoneRotation(
    Bone=dyn_legFin_01_r
    Weight=1
    bPropagateToChildren=True
    Rotation=/Script/RigVM.RigVMFunction_AccumulateQuatLerp(
     Blend=0.5
     InitialValue=(X=0,Y=0,Z=0,W=1)
     TargetValue=/Script/RigVM.RigVMFunction_MathQuaternionSlerp(A=(X=0,Y=0,Z=0,W=1), T=0.7, B=@n2.Transform.Rotation) {is=struct RigVMUnitNode, template="Interpolate::Execute(in A,in B,in T,out Result)"}.Result
    ) {is=struct RigVMUnitNode, template="AccumulateLerp::Execute(in TargetValue,in InitialValue,in Blend,in bIntegrateDeltaTime,out Result)"}.Result
    ExecutePin=@n1.ExecutePin
   ) {is=struct RigVMUnitNode}
   > &RigUnit_SetBoneRotation(
    Bone=dyn_legFin_02_r
    Weight=1
    bPropagateToChildren=True
    Rotation=&RigVMFunction_AccumulateQuatLerp(
     Blend=0.35
     InitialValue=(X=0,Y=0,Z=0,W=1)
     TargetValue=&RigVMFunction_MathQuaternionSlerp(A=(X=0,Y=0,Z=0,W=1), T=0.6, B=@n2.Transform.Rotation) {template="Interpolate::Execute(in A,in B,in T,out Result)"}.Result
    ) {template="AccumulateLerp::Execute(in TargetValue,in InitialValue,in Blend,in bIntegrateDeltaTime,out Result)"}.Result
   )
   > &RigUnit_SetBoneRotation(
    Bone=dyn_legFin_03_r
    Weight=1
    bPropagateToChildren=True
    Rotation=&RigVMFunction_AccumulateQuatLerp(
     Blend=0.2
     InitialValue=(X=0,Y=0,Z=0,W=1)
     TargetValue=&RigVMFunction_MathQuaternionSlerp(A=(X=0,Y=0,Z=0,W=1), T=0.6, B=@n2.Transform.Rotation) {template="Interpolate::Execute(in A,in B,in T,out Result)"}.Result
    ) {template="AccumulateLerp::Execute(in TargetValue,in InitialValue,in Blend,in bIntegrateDeltaTime,out Result)"}.Result
   )
   n3 = &RigUnit_GetBoneTransform(Bone=dyn_legFinTop_03_l, Space=LocalSpace)
   > &RigUnit_SetBoneRotation(
    Bone=dyn_legFin_01_l
    Weight=1
    bPropagateToChildren=True
    Rotation=&RigVMFunction_AccumulateQuatLerp(
     Blend=0.5
     InitialValue=(X=0,Y=0,Z=0,W=1)
     TargetValue=&RigVMFunction_MathQuaternionSlerp(A=(X=0,Y=0,Z=0,W=1), T=0.7, B=@n3.Transform.Rotation) {template="Interpolate::Execute(in A,in B,in T,out Result)"}.Result
    ) {template="AccumulateLerp::Execute(in TargetValue,in InitialValue,in Blend,in bIntegrateDeltaTime,out Result)"}.Result
   )
   > &RigUnit_SetBoneRotation(
    Bone=dyn_legFin_02_l
    Weight=1
    bPropagateToChildren=True
    Rotation=&RigVMFunction_AccumulateQuatLerp(
     Blend=0.35
     InitialValue=(X=0,Y=0,Z=0,W=1)
     TargetValue=&RigVMFunction_MathQuaternionSlerp(A=(X=0,Y=0,Z=0,W=1), T=0.6, B=@n3.Transform.Rotation) {template="Interpolate::Execute(in A,in B,in T,out Result)"}.Result
    ) {template="AccumulateLerp::Execute(in TargetValue,in InitialValue,in Blend,in bIntegrateDeltaTime,out Result)"}.Result
   )
   > &RigUnit_SetBoneRotation(
    Bone=dyn_legFin_03_l
    Weight=1
    bPropagateToChildren=True
    Rotation=&RigVMFunction_AccumulateQuatLerp(
     Blend=0.2
     InitialValue=(X=0,Y=0,Z=0,W=1)
     TargetValue=&RigVMFunction_MathQuaternionSlerp(A=(X=0,Y=0,Z=0,W=1), T=0.6, B=@n3.Transform.Rotation) {template="Interpolate::Execute(in A,in B,in T,out Result)"}.Result
    ) {template="AccumulateLerp::Execute(in TargetValue,in InitialValue,in Blend,in bIntegrateDeltaTime,out Result)"}.Result
   )
   RigUnit_BeginExecution = &RigUnit_BeginExecution()
   /Script/ControlRig.RigUnit_AimBone(
    Bone=dyn_armFin_03_Base_r
    DebugSettings=(Scale=10,WorldOffset=(Rotation=(X=0,Y=0,Z=0,W=1),Translation=(X=0,Y=0,Z=0),Scale3D=(X=1,Y=1,Z=1)))
    Primary=(Weight=1,Axis=(X=1,Y=0,Z=0),Target=(X=0,Y=-2.5,Z=0),Kind=Location,Space="dyn_armFin_02_r")
    Secondary=(Weight=1,Axis=(X=0,Y=0,Z=1),Target=(X=0,Y=0,Z=0),Kind=Location,Space="dyn_armFin_03_Base_r")
    Weight=1
    bPropagateToChildren=True
    ExecutePin=@RigUnit_BeginExecution.ExecutePin
   ) {is=struct RigVMUnitNode}
   n4 = &RigUnit_GetBoneTransform(Bone=dyn_armFin_01_r, Space=LocalSpace)
   > &RigUnit_SetBoneRotation(
    Bone=dyn_armFin_02_r
    Weight=1
    bPropagateToChildren=True
    Rotation=&RigVMFunction_AccumulateQuatLerp(
     Blend=0.5
     InitialValue=(X=0,Y=0,Z=0,W=1)
     TargetValue=&RigVMFunction_MathQuaternionSlerp(A=(X=0,Y=0,Z=0,W=1), T=0.725, B=@n4.Transform.Rotation) {template="Interpolate::Execute(in A,in B,in T,out Result)"}.Result
    ) {template="AccumulateLerp::Execute(in TargetValue,in InitialValue,in Blend,in bIntegrateDeltaTime,out Result)"}.Result
   )
   > &RigUnit_SetBoneRotation(
    Bone=dyn_armFin_03_r
    Weight=1
    bPropagateToChildren=True
    Rotation=&RigVMFunction_AccumulateQuatLerp(
     Blend=0.35
     InitialValue=(X=0,Y=0,Z=0,W=1)
     TargetValue=&RigVMFunction_MathQuaternionSlerp(A=(X=0,Y=0,Z=0,W=1), T=0.8, B=@n4.Transform.Rotation) {template="Interpolate::Execute(in A,in B,in T,out Result)"}.Result
    ) {template="AccumulateLerp::Execute(in TargetValue,in InitialValue,in Blend,in bIntegrateDeltaTime,out Result)"}.Result
   )
   > &RigUnit_SetBoneRotation(
    Bone=dyn_armFin_04_r
    Weight=1
    bPropagateToChildren=True
    Rotation=&RigVMFunction_AccumulateQuatLerp(
     Blend=0.2
     InitialValue=(X=0,Y=0,Z=0,W=1)
     TargetValue=&RigVMFunction_MathQuaternionSlerp(A=(X=0,Y=0,Z=0,W=1), T=0.675, B=@n4.Transform.Rotation) {template="Interpolate::Execute(in A,in B,in T,out Result)"}.Result
    ) {template="AccumulateLerp::Execute(in TargetValue,in InitialValue,in Blend,in bIntegrateDeltaTime,out Result)"}.Result
   )
   > &RigUnit_SetBoneRotation(
    Bone=dyn_armFin_05_r
    Weight=1
    bPropagateToChildren=True
    Rotation=&RigVMFunction_AccumulateQuatLerp(
     Blend=0.2
     InitialValue=(X=0,Y=0,Z=0,W=1)
     TargetValue=&RigVMFunction_MathQuaternionSlerp(A=(X=0,Y=0,Z=0,W=1), T=0.65, B=@n4.Transform.Rotation) {template="Interpolate::Execute(in A,in B,in T,out Result)"}.Result
    ) {template="AccumulateLerp::Execute(in TargetValue,in InitialValue,in Blend,in bIntegrateDeltaTime,out Result)"}.Result
   )
   > &RigUnit_AimBone(
    Bone=dyn_armFin_03_Base_l
    DebugSettings=(Scale=10,WorldOffset=(Rotation=(X=0,Y=0,Z=0,W=1),Translation=(X=0,Y=0,Z=0),Scale3D=(X=1,Y=1,Z=1)))
    Primary=(Weight=1,Axis=(X=-1,Y=0,Z=0),Target=(X=0,Y=2.7,Z=0),Kind=Location,Space="dyn_armFin_02_l")
    Secondary=(Weight=1,Axis=(X=0,Y=0,Z=1),Target=(X=0,Y=0,Z=0),Kind=Location,Space="dyn_armFin_03_Base_l")
    Weight=1
    bPropagateToChildren=True
   )
   n5 = &RigUnit_GetBoneTransform(Bone=dyn_armFin_01_l, Space=LocalSpace)
   > &RigUnit_SetBoneRotation(
    Bone=dyn_armFin_02_l
    Weight=1
    bPropagateToChildren=True
    Rotation=&RigVMFunction_AccumulateQuatLerp(
     Blend=0.5
     InitialValue=(X=0,Y=0,Z=0,W=1)
     TargetValue=&RigVMFunction_MathQuaternionSlerp(A=(X=0,Y=0,Z=0,W=1), T=0.725, B=@n5.Transform.Rotation) {template="Interpolate::Execute(in A,in B,in T,out Result)"}.Result
    ) {template="AccumulateLerp::Execute(in TargetValue,in InitialValue,in Blend,in bIntegrateDeltaTime,out Result)"}.Result
   )
   > &RigUnit_SetBoneRotation(
    Bone=dyn_armFin_03_l
    Weight=1
    bPropagateToChildren=True
    Rotation=&RigVMFunction_AccumulateQuatLerp(
     Blend=0.35
     InitialValue=(X=0,Y=0,Z=0,W=1)
     TargetValue=&RigVMFunction_MathQuaternionSlerp(A=(X=0,Y=0,Z=0,W=1), T=0.8, B=@n5.Transform.Rotation) {template="Interpolate::Execute(in A,in B,in T,out Result)"}.Result
    ) {template="AccumulateLerp::Execute(in TargetValue,in InitialValue,in Blend,in bIntegrateDeltaTime,out Result)"}.Result
   )
   > &RigUnit_SetBoneRotation(
    Bone=dyn_armFin_04_l
    Weight=1
    bPropagateToChildren=True
    Rotation=&RigVMFunction_AccumulateQuatLerp(
     Blend=0.2
     InitialValue=(X=0,Y=0,Z=0,W=1)
     TargetValue=&RigVMFunction_MathQuaternionSlerp(A=(X=0,Y=0,Z=0,W=1), T=0.675, B=@n5.Transform.Rotation) {template="Interpolate::Execute(in A,in B,in T,out Result)"}.Result
    ) {template="AccumulateLerp::Execute(in TargetValue,in InitialValue,in Blend,in bIntegrateDeltaTime,out Result)"}.Result
   )
   > &RigUnit_SetBoneRotation(
    Bone=dyn_armFin_05_l
    Weight=1
    bPropagateToChildren=True
    Rotation=&RigVMFunction_AccumulateQuatLerp(
     Blend=0.2
     InitialValue=(X=0,Y=0,Z=0,W=1)
     TargetValue=&RigVMFunction_MathQuaternionSlerp(A=(X=0,Y=0,Z=0,W=1), T=0.65, B=@n5.Transform.Rotation) {template="Interpolate::Execute(in A,in B,in T,out Result)"}.Result
    ) {template="AccumulateLerp::Execute(in TargetValue,in InitialValue,in Blend,in bIntegrateDeltaTime,out Result)"}.Result
   )
   /Script/ControlRig.RigUnit_BoneName(Bone=dyn_armFin_03_Base_r) {is=struct RigVMUnitNode}
   &RigUnit_GetBoneTransform(Bone=None, Space=LocalSpace)
   /Script/ControlRig.RigUnit_GetInitialBoneTransform(Bone=None, Space=GlobalSpace) {is=struct RigVMUnitNode}
   &RigUnit_GetInitialBoneTransform(Space=GlobalSpace)
   &RigUnit_GetRelativeBoneTransform(Bone=None, Space=None)
   n6 = &RigVMFunction_MathTransformMakeRelative(
    Global=&RigUnit_GetBoneTransform(Bone=None).Transform
    Parent=&RigUnit_GetInitialBoneTransform(Bone=None, Space=GlobalSpace).Transform
   )
   &RigVMFunction_MathFloatMul(B=10, A=&RigVMFunction_MathQuaternionToEuler(RotationOrder=ZYX, Value=@n6.Local.Rotation).Result.Z) {template="Multiply::Execute(in A,in B,out Result)"}
   &RigVMFunction_MathQuaternionFromEuler(Euler=(X=0,Y=0,Z=0))
   /Script/RigVM.RigVMFunction_MathQuaternionFromTwoVectors(B=(X=1,Y=0,Z=0), A=@n6.Local.Translation) {is=struct RigVMUnitNode, template="FromTwoVectors::Execute(in A,in B,out Result)"}
   &RigVMFunction_MathQuaternionSlerp(A=(X=0,Y=0,Z=0,W=1), T=0.125, B=@n2.Transform.Rotation) {template="Interpolate::Execute(in A,in B,in T,out Result)"}
   &RigVMFunction_MathQuaternionSlerp(A=(X=0,Y=0,Z=0,W=1), T=0.125, B=@n3.Transform.Rotation) {template="Interpolate::Execute(in A,in B,in T,out Result)"}
   &RigVMFunction_MathQuaternionSlerp(A=(X=0,Y=0,Z=0,W=1), B=@n4.Transform.Rotation) {template="Interpolate::Execute(in A,in B,in T,out Result)"}
   &RigVMFunction_MathQuaternionSlerp(A=(X=0,Y=0,Z=0,W=1), B=@n5.Transform.Rotation) {template="Interpolate::Execute(in A,in B,in T,out Result)"}
   n7 = &RigVMFunction_MathQuaternionFromEuler(Euler=(X=0,Y=0,Z=0), RotationOrder=XYZ)
   &RigUnit_SetBoneRotation(Bone=dyn_armFin_03_Base_r, ExecutePin=(), Weight=1, bPropagateToChildren=false, Rotation=@n7.Result)
   n8 = &RigUnit_GetBoneTransform(Bone=dyn_armFin_03_Base_r)
   n9 = /Script/RigVM.RigVMFunction_MathQuaternionGetAxis(Axis=X, Quaternion=@n8.Transform.Rotation) {is=struct RigVMUnitNode, template="Axis::Execute(in Quaternion,in Axis,out Result)"}
   n10 = &RigUnit_GetBoneTransform(Bone=dyn_armFin_02_r)
   n11 = /Script/RigVM.RigVMFunction_MathVectorSub(
    B=@n8.Transform.Translation
    A=/Script/RigVM.RigVMFunction_MathTransformTransformVector(Location=(X=0,Y=-3,Z=0), Transform=@n10.Transform) {is=struct RigVMUnitNode}.Result
   ) {is=struct RigVMUnitNode, template="Subtract::Execute(in A,in B,out Result)"}
   &RigUnit_SetBoneRotation(
    Bone=dyn_armFin_03_Base_r
    ExecutePin=()
    Space=GlobalSpace
    Weight=1
    bPropagateToChildren=True
    Rotation=/Script/RigVM.RigVMFunction_MathQuaternionMul(
     B=@n8.Transform.Rotation
     A=&RigVMFunction_MathQuaternionFromTwoVectors(A=@n9.Result, B=@n11.Result) {template="FromTwoVectors::Execute(in A,in B,out Result)"}.Result
    ) {is=struct RigVMUnitNode, template="Multiply::Execute(in A,in B,out Result)"}.Result
   )
   n12 = &RigVMFunction_MathQuaternionGetAxis(Axis=X, Quaternion=@n10.Transform.Rotation) {template="Axis::Execute(in Quaternion,in Axis,out Result)"}
   n13 = &RigVMFunction_MathVectorSub(
    B=@n10.Transform.Translation
    A=&RigVMFunction_MathTransformTransformVector(Location=(X=0,Y=0,Z=0), Transform=@n8.Transform).Result
   ) {template="Subtract::Execute(in A,in B,out Result)"}
   &RigUnit_SetBoneRotation(
    Bone=dyn_piston_LegBody_r
    ExecutePin=()
    Space=GlobalSpace
    Weight=1
    bPropagateToChildren=True
    Rotation=&RigVMFunction_MathQuaternionMul(
     B=@n10.Transform.Rotation
     A=&RigVMFunction_MathQuaternionFromTwoVectors(A=@n12.Result, B=@n13.Result) {template="FromTwoVectors::Execute(in A,in B,out Result)"}.Result
    ) {template="Multiply::Execute(in A,in B,out Result)"}.Result
   )
   /Script/ControlRig.RigUnit_SetBoneTransform(
    Bone=None
    ExecutePin=()
    Space=LocalSpace
    Transform=(Rotation=(X=0,Y=0,Z=0,W=1),Translation=(X=0,Y=0,Z=0),Scale3D=(X=1,Y=1,Z=1))
    Weight=1
    bPropagateToChildren=false
   ) {is=struct RigVMUnitNode}
   /Script/ControlRig.RigUnit_VisualDebugTransform(
    DebugDrawSettings=(DepthPriority=SDPG_Foreground,Lifetime=-1)
    Scale=10
    Value=(Rotation=(X=0,Y=0,Z=0,W=1),Translation=(X=0,Y=0,Z=0),Scale3D=(X=1,Y=1,Z=1))
    bEnabled=True
    Value.Translation=&RigUnit_GetBoneTransform(Bone=None, Space=LocalSpace).Transform.Translation
    Value.Rotation=&RigVMFunction_MathQuaternionMul(B=@n7.Result, A=@n6.Local.Rotation) {template="Multiply::Execute(in A,in B,out Result)"}.Result
   ) {is=struct RigVMUnitNode}
  ]}
]
hierarchy [
 root {loc=<0 0 0>, rot=<0 0 0 1>}
 pelvis {parent=root, loc=<0 -0.5301 98.0571>, rot=<0.0127 0.707 -0.0127 -0.707>}
 spine_01 {parent=pelvis, loc=<6.1208 -0.9328 0>, rot=<0 0 0.034 -0.9994>}
 spine_02 {parent=spine_01, loc=<6.5996 -0.2095 0>, rot=<0 0 0.0238 -0.9997>}
 spine_03 {parent=spine_02, loc=<6.4914 -0.1237 0>, rot=<0 0 0.0149 -0.9999>}
 spine_04 {parent=spine_03, loc=<6.9598 0.0794 0>, rot=<0 0 -0.0132 -0.9999>}
 spine_05 {parent=spine_04, loc=<6.9987 0 0>, rot=<0 0 -0.0015 -1>}
 clavicle_l {parent=spine_05, loc=<5.4831 -0.1053 -3.2946>, rot=<0.0465 -0.7289 -0.009 -0.6829>}
 upperarm_l {parent=clavicle_l, loc=<9.8658 -1.2136 -0.4105>, rot=<0.0164 -0.3245 -0.0269 -0.9454>}
 lowerarm_l {parent=upperarm_l, loc=<26.465 -0.8345 0.259>, rot=<0.0036 0.0161 -0.2298 -0.9731>}
 hand_l {parent=lowerarm_l, loc=<22.7355 -0.3203 -0.6105>, rot=<-0.6784 -0.0223 0.0735 -0.7307>}
 index_metacarpal_l {parent=hand_l, loc=<1.6535 0.0183 -2.96>, rot=<0.0158 -0.0124 -0.0173 -0.9996>}
 index_01_l {parent=index_metacarpal_l, loc=<8.3008 0.0357 -0.0524>, rot=<-0.0044 0.0485 0.0924 -0.9945>}
 index_02_l {parent=index_01_l, loc=<3.3587 -0.3003 0.0452>, rot=<-0.002 0.0027 0.3225 -0.9466>}
 index_03_l {parent=index_02_l, loc=<2.5578 -0.123 -0.0252>, rot=<-0.002 0.0067 0.2558 -0.9667>}
 middle_metacarpal_l {parent=hand_l, loc=<2.1856 -0.1725 -0.5984>, rot=<0.006 0.0168 -0.0329 -0.9993>}
 middle_01_l {parent=middle_metacarpal_l, loc=<7.3275 -0.2447 0.0129>, rot=<-0.0208 0.0042 0.0858 -0.9961>}
 middle_02_l {parent=middle_01_l, loc=<4.0123 0.1401 0.1175>, rot=<-0.0218 0.0258 0.3512 -0.9357>}
 middle_03_l {parent=middle_02_l, loc=<2.6077 0.0534 0.1373>, rot=<0.0082 0.0188 0.1708 -0.9851>}
 pinky_metacarpal_l {parent=hand_l, loc=<1.7616 -0.9753 3.3814>, rot=<-0.0397 0.0814 -0.069 -0.9935>}
 pinky_01_l {parent=pinky_metacarpal_l, loc=<7.1709 -0.4831 -0.1094>, rot=<-0.0654 -0.0035 0.2242 -0.9723>}
 pinky_02_l {parent=pinky_01_l, loc=<2.8274 0.0893 -0.1043>, rot=<0.0037 0.0011 0.2854 -0.9584>}
 pinky_03_l {parent=pinky_02_l, loc=<2.2046 0.0308 -0.0078>, rot=<-0.0002 0.0098 0.1445 -0.9895>}
 ring_metacarpal_l {parent=hand_l, loc=<2.2074 -0.3625 1.3638>, rot=<-0.0284 0.0722 -0.0458 -0.9959>}
 ring_01_l {parent=ring_metacarpal_l, loc=<7.0295 -0.6197 -0.1446>, rot=<-0.0489 -0.0252 0.1892 -0.9804>}
 ring_02_l {parent=ring_01_l, loc=<3.6174 0.4647 -0.1027>, rot=<0 0 0.2806 -0.9598>}
 ring_03_l {parent=ring_02_l, loc=<2.4152 0.1141 0.0199>, rot=<0 0 0.1189 -0.9929>}
 thumb_01_l {parent=hand_l, loc=<1.3118 -0.3191 -2.459>, rot=<0.6546 -0.4042 0.0664 -0.6354>}
 thumb_02_l {parent=thumb_01_l, loc=<5.9336 -0.0448 0.0401>, rot=<-0.0037 0.0213 0.298 -0.9543>}
 thumb_03_l {parent=thumb_02_l, loc=<2.2223 0.3682 -0.0907>, rot=<0.0118 -0.013 0.0066 -0.9998>}
 lowerarm_twist_01_l {parent=lowerarm_l, loc=<18.316 0 0>, rot=<0 0 0 1>}
 lowerarm_twist_02_l {parent=lowerarm_l, loc=<10.6707 0 0>, rot=<0 0 0 1>}
 dyn_armFin_05_Base_l {parent=lowerarm_twist_02_l, loc=<-2.4669 -3.3971 0.6743>, rot=<-0.0706 -0.0379 0.3282 -0.9412>}
 dyn_armFin_05_l {parent=dyn_armFin_05_Base_l, loc=<0 0 0>, rot=<0 0 0 -1>}
 dyn_armFin_04_Base_l {parent=lowerarm_twist_02_l, loc=<-7.0884 -3.883 0.671>, rot=<-0.0106 -0.0347 0.3653 -0.9302>}
 dyn_armFin_04_l {parent=dyn_armFin_04_Base_l, loc=<0 0 0>, rot=<0 0 0 -1>}
 dyn_armFin_03_Base_l {parent=lowerarm_l, loc=<-1.9725 -3.1522 0.6871>, rot=<0.0132 -0.0184 0.4054 -0.9138>}
 dyn_armFin_03_l {parent=dyn_armFin_03_Base_l, loc=<0 0 0>, rot=<0 0 0 1>}
 upperarm_twist_01_l {parent=upperarm_l, loc=<7.2135 -0.1134 0>, rot=<0 0 0 -1>}
 upperarm_twist_02_l {parent=upperarm_l, loc=<19.7998 -0.1134 0>, rot=<0 0 0 -1>}
 dyn_armFin_01_Base_l {parent=upperarm_twist_02_l, loc=<-4.1156 -5.6092 0.5548>, rot=<0.0612 0.0247 0.2688 -0.9609>}
 dyn_armFin_01_l {parent=dyn_armFin_01_Base_l, loc=<0 0 0>, rot=<0 0 0 1>}
 dyn_armFin_02_Base_l {parent=upperarm_twist_02_l, loc=<-0.5536 -4.7805 0.5425>, rot=<0.031 0.0071 0.2279 -0.9732>}
 dyn_armFin_02_l {parent=dyn_armFin_02_Base_l, loc=<0 0 0>, rot=<0 0 0 1>}
 clavicle_r {parent=spine_05, loc=<5.4831 -0.1053 3.2946>, rot=<0.7289 0.0465 -0.6829 0.009>}
 upperarm_r {parent=clavicle_r, loc=<-9.8658 1.2136 0.4104>, rot=<0.0164 -0.3245 -0.0269 -0.9454>}
 lowerarm_r {parent=upperarm_r, loc=<-26.4651 0.8345 -0.259>, rot=<0.0036 0.0161 -0.2298 -0.9731>}
 hand_r {parent=lowerarm_r, loc=<-22.7356 0.3203 0.6106>, rot=<-0.6784 -0.0223 0.0735 -0.7307>}
 index_metacarpal_r {parent=hand_r, loc=<-1.6534 -0.0181 2.9601>, rot=<0.0158 -0.0124 -0.0173 -0.9996>}
 index_01_r {parent=index_metacarpal_r, loc=<-8.3008 -0.0357 0.0524>, rot=<-0.0044 0.0485 0.0924 -0.9945>}
 index_02_r {parent=index_01_r, loc=<-3.3589 0.3003 -0.0453>, rot=<-0.002 0.0027 0.3225 -0.9466>}
 index_03_r {parent=index_02_r, loc=<-2.5578 0.123 0.0252>, rot=<-0.002 0.0067 0.2558 -0.9667>}
 middle_metacarpal_r {parent=hand_r, loc=<-2.1856 0.1726 0.5984>, rot=<0.006 0.0168 -0.0329 -0.9993>}
 middle_01_r {parent=middle_metacarpal_r, loc=<-7.3273 0.2447 -0.0129>, rot=<-0.0208 0.0042 0.0858 -0.9961>}
 middle_02_r {parent=middle_01_r, loc=<-4.0124 -0.14 -0.1175>, rot=<-0.0218 0.0258 0.3512 -0.9357>}
 middle_03_r {parent=middle_02_r, loc=<-2.6077 -0.0533 -0.1373>, rot=<0.0082 0.0188 0.1708 -0.9851>}
 pinky_metacarpal_r {parent=hand_r, loc=<-1.7617 0.9753 -3.3814>, rot=<-0.0397 0.0814 -0.069 -0.9935>}
 pinky_01_r {parent=pinky_metacarpal_r, loc=<-7.1707 0.4832 0.1094>, rot=<-0.0654 -0.0035 0.2242 -0.9723>}
 pinky_02_r {parent=pinky_01_r, loc=<-2.8274 -0.0892 0.1043>, rot=<0.0037 0.0011 0.2854 -0.9584>}
 pinky_03_r {parent=pinky_02_r, loc=<-2.2045 -0.031 0.0078>, rot=<-0.0002 0.0098 0.1445 -0.9895>}
 ring_metacarpal_r {parent=hand_r, loc=<-2.2073 0.3627 -1.3638>, rot=<-0.0284 0.0722 -0.0458 -0.9959>}
 ring_01_r {parent=ring_metacarpal_r, loc=<-7.0297 0.6197 0.1446>, rot=<-0.0489 -0.0252 0.1892 -0.9804>}
 ring_02_r {parent=ring_01_r, loc=<-3.6172 -0.4647 0.1027>, rot=<0 0 0.2806 -0.9598>}
 ring_03_r {parent=ring_02_r, loc=<-2.4152 -0.1141 -0.0199>, rot=<0 0 0.1189 -0.9929>}
 thumb_01_r {parent=hand_r, loc=<-1.3119 0.3192 2.459>, rot=<0.6546 -0.4042 0.0664 -0.6354>}
 thumb_02_r {parent=thumb_01_r, loc=<-5.9336 0.0448 -0.0401>, rot=<-0.0037 0.0213 0.298 -0.9543>}
 thumb_03_r {parent=thumb_02_r, loc=<-2.2223 -0.3682 0.0907>, rot=<0.0118 -0.013 0.0066 -0.9998>}
 lowerarm_twist_01_r {parent=lowerarm_r, loc=<-18.3161 0 0>, rot=<0 0 0 1>}
 lowerarm_twist_02_r {parent=lowerarm_r, loc=<-10.6707 0 0>, rot=<0 0 0 1>}
 dyn_armFin_04_Base_r {parent=lowerarm_twist_02_r, loc=<7.0887 3.8829 -0.6712>, rot=<-0.0106 -0.0347 0.3653 -0.9302>}
 dyn_armFin_04_r {parent=dyn_armFin_04_Base_r, loc=<0 0 0>, rot=<0 0 0 1>}
 dyn_armFin_05_Base_r {parent=lowerarm_twist_02_r, loc=<2.4667 3.3972 -0.674>, rot=<-0.0706 -0.0379 0.3282 -0.9412>}
 dyn_armFin_05_r {parent=dyn_armFin_05_Base_r, loc=<0 0 0>, rot=<0 0 0 1>}
 dyn_armFin_03_Base_r {parent=lowerarm_r, loc=<1.9728 3.1521 -0.6874>, rot=<0.0132 -0.0184 0.4054 -0.9138>}
 dyn_armFin_03_r {parent=dyn_armFin_03_Base_r, loc=<0 0 0>, rot=<0 0 0 1>}
 upperarm_twist_01_r {parent=upperarm_r, loc=<-7.2135 0.1134 0>, rot=<0 0 0 1>}
 upperarm_twist_02_r {parent=upperarm_r, loc=<-19.7998 0.1134 0>, rot=<0 0 0 1>}
 dyn_armFin_01_Base_r {parent=upperarm_twist_02_r, loc=<4.1154 5.6092 -0.5546>, rot=<0.0612 0.0247 0.2688 -0.9609>}
 dyn_armFin_01_r {parent=dyn_armFin_01_Base_r, loc=<0 0 0>, rot=<0 0 0 1>}
 dyn_armFin_02_Base_r {parent=upperarm_twist_02_r, loc=<0.5535 4.7805 -0.5423>, rot=<0.031 0.0071 0.2279 -0.9732>}
 dyn_armFin_02_r {parent=dyn_armFin_02_Base_r, loc=<0 0 0>, rot=<0 0 0 1>}
 neck_01 {parent=spine_05, loc=<11.3138 0.1529 0>, rot=<0 0 -0.2003 -0.9797>}
 neck_02 {parent=neck_01, loc=<4.5627 0 0>, rot=<0 0 0.0504 -0.9987>}
 head {parent=neck_02, loc=<5.3914 0 0>, rot=<0 0 0.115 -0.9934>}
 dyn_hairMain_01_Base {parent=head, loc=<6.2693 -7.9101 0>, rot=<0.0209 -0.9998 0 0>}
 dyn_hairMain_01_DRV {parent=dyn_hairMain_01_Base, loc=<0 0 0>, rot=<0 0 0 -1>}
 dyn_hairMain_01 {parent=dyn_hairMain_01_DRV, loc=<0 0 0>, rot=<0 0 0 -1>}
 dyn_hairMain_02 {parent=dyn_hairMain_01, loc=<13.4296 0 0>, rot=<0 0 0.1057 -0.9944>}
 dyn_hairMain_03 {parent=dyn_hairMain_02, loc=<12.6863 0 0>, rot=<0 0 -0.064 -0.9979>}
 dyn_hairMain_04 {parent=dyn_hairMain_03, loc=<12.7071 0 0>, rot=<0 0 -0.0382 -0.9993>}
 dyn_hairMain_05 {parent=dyn_hairMain_04, loc=<12.8002 0 0>, rot=<0 0 -0.0119 -0.9999>}
 dyn_hairMain_06 {parent=dyn_hairMain_05, loc=<12.8476 0 0>, rot=<0 0 -0.0045 -1>}
 dyn_hairMain_07 {parent=dyn_hairMain_06, loc=<12.3747 0 0>, rot=<0 0 0.0004 -1>}
 dyn_hairMain_08 {parent=dyn_hairMain_07, loc=<8.6504 0 0>, rot=<0 0 -0.0053 -1>}
 dyn_hairMain_09 {parent=dyn_hairMain_08, loc=<15.0041 0 0>, rot=<0 0 0 -1>}
 thigh_l {parent=pelvis, loc=<-1.0855 -0.2109 -8.7053>, rot=<0.0009 0.0355 0.0155 -0.9993>}
 calf_l {parent=thigh_l, loc=<-42.5284 -0.0019 -0.7297>, rot=<0.0006 -0.0161 -0.0391 -0.9991>}
 calf_twist_01_l {parent=calf_l, loc=<-35.8483 -1.9197 -0.5794>, rot=<0.0237 0.0095 -0.0035 -0.9997>}
 dyn_Base_l1 {parent=calf_twist_01_l, loc=<-1.461 -3.8505 -2.7323>, rot=<-0.2457 -0.7459 -0.6103 0.1037>}
 dyn_legFin_02_l {parent=dyn_Base_l1, loc=<0 0 0>, rot=<0 0 0 -1>}
 dyn_Base_l2 {parent=calf_twist_01_l, loc=<-8.1597 -2.762 -1.6227>, rot=<-0.2795 -0.7243 -0.6174 0.1273>}
 dyn_legFin_03_l {parent=dyn_Base_l2, loc=<0 0 0>, rot=<0 0 0 1>}
 calf_twist_02_l {parent=calf_l, loc=<-19.1546 -1.7943 -0.2638>, rot=<0.0237 0.0095 -0.0035 -0.9997>}
 dyn_legFinTop_01_Base_l {parent=calf_twist_02_l, loc=<1.7019 -6.9401 -4.509>, rot=<-0.1446 -0.7834 -0.5934 0.115>}
 dyn_legFinTop_01_l {parent=dyn_legFinTop_01_Base_l, loc=<0 0 0>, rot=<0 0 0 1>}
 dyn_legFinTop_02_l {parent=dyn_legFinTop_01_l, loc=<-9.9534 0 0>, rot=<-0.0258 -0.003 0.1098 -0.9936>}
 dyn_legFinTop_03_l {parent=dyn_legFinTop_02_l, loc=<-6.4096 0 0>, rot=<0.0439 0.0475 0.1128 -0.9915>}
 dyn_Base_l {parent=calf_twist_02_l, loc=<-7.5534 -5.3872 -3.7419>, rot=<-0.231 -0.7956 -0.5539 0.0824>}
 dyn_legFin_01_l {parent=dyn_Base_l, loc=<0 0 0>, rot=<0 0 0 1>}
 foot_l {parent=calf_l, loc=<-46.7401 -2.5014 -0.592>, rot=<0.0427 -0.0238 0.0353 -0.9982>}
 ball_l {parent=foot_l, loc=<-3.5658 10.8066 -0.1544>, rot=<0.0001 -0.0001 -0.7063 -0.7079>}
 thigh_twist_01_l {parent=thigh_l, loc=<-27.952 0 0>, rot=<0.0013 0 0 -1>}
 dyn_thighScale_l {parent=thigh_l, loc=<0 0 0>, rot=<0 0 0 -1>}
 thigh_r {parent=pelvis, loc=<-1.0853 -0.2109 8.7053>, rot=<0.0355 -0.0009 0.9993 0.0155>}
 calf_r {parent=thigh_r, loc=<42.5286 0.0019 0.7297>, rot=<0.0006 -0.0161 -0.0391 -0.9991>}
 calf_twist_01_r {parent=calf_r, loc=<35.8485 1.9197 0.5794>, rot=<0.0237 0.0095 -0.0035 -0.9997>}
 dyn_legFin_02_Base_r {parent=calf_twist_01_r, loc=<1.4609 3.8506 2.7323>, rot=<-0.2457 -0.7459 -0.6103 0.1037>}
 dyn_legFin_02_r {parent=dyn_legFin_02_Base_r, loc=<0 0 0>, rot=<0 0 0 -1>}
 dyn_legFin_03_Base_r {parent=calf_twist_01_r, loc=<8.1595 2.762 1.6227>, rot=<-0.2795 -0.7243 -0.6174 0.1273>}
 dyn_legFin_03_r {parent=dyn_legFin_03_Base_r, loc=<0 0 0>, rot=<0 0 0 -1>}
 calf_twist_02_r {parent=calf_r, loc=<19.1546 1.7943 0.2638>, rot=<0.0237 0.0095 -0.0035 -0.9997>}
 dyn_legFinTop_01_Base_r {parent=calf_twist_02_r, loc=<-1.702 6.9401 4.5089>, rot=<-0.1446 -0.7834 -0.5934 0.115>}
 dyn_legFinTop_01_r {parent=dyn_legFinTop_01_Base_r, loc=<0 0 0>, rot=<0 0 0 -1>}
 dyn_legFinTop_02_r {parent=dyn_legFinTop_01_r, loc=<9.9535 0 0>, rot=<-0.0258 -0.003 0.1098 -0.9936>}
 dyn_legFinTop_03_r {parent=dyn_legFinTop_02_r, loc=<6.4096 0 0>, rot=<0.0439 0.0475 0.1128 -0.9915>}
 dyn_legFin_01_Base_r {parent=calf_twist_02_r, loc=<7.5533 5.3872 3.7419>, rot=<-0.231 -0.7956 -0.5539 0.0824>}
 dyn_legFin_01_r {parent=dyn_legFin_01_Base_r, loc=<0 0 0>, rot=<0 0 0 -1>}
 foot_r {parent=calf_r, loc=<46.7403 2.5014 0.592>, rot=<0.0427 -0.0238 0.0353 -0.9982>}
 ball_r {parent=foot_r, loc=<3.5658 -10.8066 0.1544>, rot=<0.0001 -0.0001 -0.7063 -0.7079>}
 thigh_twist_01_r {parent=thigh_r, loc=<27.952 0 0>, rot=<0.0013 0 0 -1>}
 dyn_thighScale_r {parent=thigh_r, loc=<0 0 0>, rot=<0 0 0 -1>}
]
