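# Physics asset for the F_MED_Angler female medium player body cosmetic, covering its dynamic arm and upper-leg fin bones.
# Mostly capsule bodies form arm fins jointed through base bones to upper-arm twist bones, and three-segment leg fin chains hanging from box base bodies.
# Tuning is asymmetric on the legs: left-chain joints are fully free in rotation, while right-chain joints, like both arm fin tips, lock twist and restrict swing.
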PhysicsAsset /BRCosmetics/Characters/Player/Female/Medium/Bodies/F_MED_Angler/Meshes/F_MED_Angler_Physics
bodies {
 dyn_armFin_01_Base_l [{shapeType=Capsule, length=7.0433, radius=1.2994, center=<8.947 -0.0151 -0.0082>, rotation=<rot 0 89.7299 -88.2578>}]
 dyn_armFin_01_Base_r [{shapeType=Capsule, length=0.505, radius=0.505, rotation=<rot 0 -90 0>}]
 dyn_armFin_01_l [{shapeType=Capsule, length=15.3248, radius=1.5764, center=<7.696 0 -0.0001>, rotation=<rot 90 0 180>}]
 dyn_armFin_01_r [{shapeType=Capsule, length=15.3248, radius=1.5764, center=<-7.696 0 0.0001>, rotation=<rot 90 0 0>}]
 dyn_legFinTop_01_Base_l [{shapeType=Box, extentX=1.01, extentY=1.01, extentZ=1.01, rotation=<rot 0 -90 0>}]
 dyn_legFinTop_01_Base_r [{shapeType=Box, extentX=1.01, extentY=1.01, extentZ=1.01, rotation=<rot 0 -90 0>}]
 dyn_legFinTop_01_l [{shapeType=Capsule, length=2.9917, radius=0.8946, center=<-7.2694 -0.3605 0.0633>, rotation=<rot 0 74.7611 -88.4885>}]
 dyn_legFinTop_01_r [{shapeType=Capsule, length=9.3908, radius=1.1659, center=<0.5746 0 -0.0003>, rotation=<rot 90 0 0>}]
 dyn_legFinTop_02_l [{shapeType=Capsule, length=2.3055, radius=0.5073, center=<-4.3188 -0.0741 0.1335>, rotation=<rot 0 -91.0637 -89.4316>}]
 dyn_legFinTop_02_r [{shapeType=Capsule, length=6.1637, radius=1.1659, center=<0.5567 -0.7654 0>, rotation=<rot 90 0 0>}]
 dyn_legFinTop_03_l [{shapeType=Capsule, length=3.3594, radius=0.8078, center=<-5.3128 -0.1908 -0.0265>, rotation=<rot 0 88.3768 -89.0064>}]
 dyn_legFinTop_03_r [{shapeType=Capsule, length=7.5502, radius=1.1659, center=<1.5864 -0.8957 0.0002>, rotation=<rot 90 0 0>}]
 upperarm_twist_02_l [{shapeType=Capsule, length=0.505, radius=0.505, rotation=<rot 0 -90 0>}]
 upperarm_twist_02_r [{shapeType=Capsule, length=0.505, radius=0.505, rotation=<rot 0 -90 0>}]
}
constraints [
 {
  bone1Name=dyn_armFin_01_Base_l
  bone2Name=upperarm_twist_02_l
  swing1LimitDegrees=45
  swing1Motion=Free
  swing2LimitDegrees=45
  swing2Motion=Free
  twistLimitDegrees=45
  twistMotion=Free
 }
 {
  bone1Name=dyn_armFin_01_Base_r
  bone2Name=upperarm_twist_02_r
  swing1LimitDegrees=45
  swing1Motion=Free
  swing2LimitDegrees=45
  swing2Motion=Free
  twistLimitDegrees=45
  twistMotion=Free
 }
 {
  bone1Name=dyn_armFin_01_r
  bone2Name=dyn_armFin_01_Base_r
  swing1LimitDegrees=30
  swing2LimitDegrees=30
  twistLimitDegrees=45
  twistMotion=Locked
 }
 {
  bone1Name=dyn_armFin_01_l
  bone2Name=dyn_armFin_01_Base_l
  swing1LimitDegrees=30
  swing2LimitDegrees=30
  twistLimitDegrees=45
  twistMotion=Locked
 }
 {
  bone1Name=dyn_legFinTop_01_l
  bone2Name=dyn_legFinTop_01_Base_l
  swing1LimitDegrees=45
  swing1Motion=Free
  swing2LimitDegrees=45
  swing2Motion=Free
  twistLimitDegrees=45
  twistMotion=Free
 }
 {
  bone1Name=dyn_legFinTop_02_l
  bone2Name=dyn_legFinTop_01_l
  swing1LimitDegrees=45
  swing1Motion=Free
  swing2LimitDegrees=45
  swing2Motion=Free
  twistLimitDegrees=45
  twistMotion=Free
 }
 {
  bone1Name=dyn_legFinTop_03_l
  bone2Name=dyn_legFinTop_02_l
  swing1LimitDegrees=45
  swing1Motion=Free
  swing2LimitDegrees=45
  swing2Motion=Free
  twistLimitDegrees=45
  twistMotion=Free
 }
 {
  bone1Name=dyn_legFinTop_01_r
  bone2Name=dyn_legFinTop_01_Base_r
  swing1LimitDegrees=8
  swing2LimitDegrees=8
  twistLimitDegrees=45
  twistMotion=Locked
 }
 {
  bone1Name=dyn_legFinTop_02_r
  bone2Name=dyn_legFinTop_01_r
  swing1LimitDegrees=12
  swing2LimitDegrees=12
  twistLimitDegrees=45
  twistMotion=Locked
 }
 {
  bone1Name=dyn_legFinTop_03_r
  bone2Name=dyn_legFinTop_02_r
  swing1LimitDegrees=12
  swing2LimitDegrees=12
  twistLimitDegrees=45
  twistMotion=Locked
 }
]
