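# Physics asset for the right leg top fin of the F_MED_Angler female player cosmetic body, defining how its dynamic fin bones can move.
# A box body on the fin base bone anchors a chain of three capsule bodies along the successive fin bones.
# Each segment is constrained to its predecessor with narrow swing limits, tighter at the root joint, and locked twist, so the fin bends slightly without rolling.
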PhysicsAsset /BRCosmetics/Characters/Player/Female/Medium/Bodies/F_MED_Angler/Meshes/F_MED_Angler_RightLegPhysics
bodies {
 dyn_legFinTop_01_Base_r [{shapeType=Box, extentX=1.01, extentY=1.01, extentZ=1.01, rotation=<rot 0 -90 0>}]
 dyn_legFinTop_01_r [{shapeType=Capsule, length=9.3908, radius=1.1659, center=<0.5746 0 -0.0003>, rotation=<rot 90 0 0>}]
 dyn_legFinTop_02_r [{shapeType=Capsule, length=6.1637, radius=1.1659, center=<0.5567 -0.7654 0>, rotation=<rot 90 0 0>}]
 dyn_legFinTop_03_r [{shapeType=Capsule, length=7.5502, radius=1.1659, center=<1.5864 -0.8957 0.0002>, rotation=<rot 90 0 0>}]
}
constraints [
 {
  bone1Name=dyn_legFinTop_01_r
  bone2Name=dyn_legFinTop_01_Base_r
  swing1LimitDegrees=8
  swing2LimitDegrees=8
  twistLimitDegrees=45
  twistMotion=Locked
 }
 {
  bone1Name=dyn_legFinTop_02_r
  bone2Name=dyn_legFinTop_01_r
  swing1LimitDegrees=12
  swing2LimitDegrees=12
  twistLimitDegrees=45
  twistMotion=Locked
 }
 {
  bone1Name=dyn_legFinTop_03_r
  bone2Name=dyn_legFinTop_02_r
  swing1LimitDegrees=12
  swing2LimitDegrees=12
  twistLimitDegrees=45
  twistMotion=Locked
 }
]
