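# Animation Blueprint for the F_MED_Angler female player body cosmetic, adding secondary physics to a pose copied, with curves, from the attached parent mesh.
# Three chained RigidBody nodes simulate in base-bone space: one from root, then right and left thigh passes with separate physics assets and much weaker gravity.
# The Angler Control Rig post-processes the simulated pose before output.
# The event graph has no logic and no variables exist.
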AnimBlueprint /BRCosmetics/Characters/Player/Female/Medium/Bodies/F_MED_Angler/Meshes/F_MED_Angler_AnimBP
target_skeleton=&/F_MED_Angler_Skeleton
graphs [
 AnimGraph {nodes [
   n1 = RigidBody(ComponentPose=Local To Component(LocalPose=Copy Pose From Mesh() {bCopyCurves=true, bUseAttachedParent=true})) {
    bClampLinearTranslationLimitToRefPose=true
    bOverrideWorldGravity=true
    overridePhysicsAsset=&/F_MED_Angler_Physics
    simulationSpace=BaseBoneSpace
    alphaBoolBlend {blendOption=Linear}
    baseBoneRef {boneName=root}
    componentLinearAccScale=<0.4 0.4 0.4>
    componentLinearVelScale=<0.2 0.2 0.2>
    overrideWorldGravity {z=-980}
   }
   n2 = RigidBody(ComponentPose=@n1) {
    bClampLinearTranslationLimitToRefPose=true
    bOverrideWorldGravity=true
    overridePhysicsAsset=&/F_MED_Angler_RightLegPhysics
    simulationSpace=BaseBoneSpace
    alphaBoolBlend {blendOption=Linear}
    baseBoneRef {boneName=thigh_r}
    componentLinearAccScale=<0.65 0.65 0.65>
    componentLinearVelScale=<0.2 0.2 0.2>
    overrideWorldGravity {z=-50}
   }
   n3 = RigidBody(ComponentPose=@n2) {
    bClampLinearTranslationLimitToRefPose=true
    bOverrideWorldGravity=true
    overridePhysicsAsset=&/F_MED_Angler_LeftLegPhysics
    simulationSpace=BaseBoneSpace
    alphaBoolBlend {blendOption=Linear}
    baseBoneRef {boneName=thigh_l}
    componentLinearAccScale=<0.65 0.65 0.65>
    componentLinearVelScale=<0.2 0.2 0.2>
    overrideWorldGravity {z=-50}
   }
   n4 = Control Rig(Source=Component To Local(ComponentPose=@n3)) {
    alphaBoolBlend {blendOption=Linear}
    controlRigAssetReference {blueprintRigClass=&/F_MED_Angler_CtrlRig.F_MED_Angler_CtrlRig_C}
   }
   Output Pose(Result=@n4) {layerGroup=None}
  ]}
 EventGraph {nodes [Event BlueprintUpdateAnimation()]}
]
variables []
