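# Physics asset for the female medium Angler cosmetic's BackpackOn mesh part, covering a dynamic main hair chain and the upper body.
# Seven hair capsules link in sequence with tight swing and twist limits, tapering in swing after the upper joints, and hang from spine_05 through two tiny free-moving driver bodies.
# Pelvis, spine and clavicle capsules form the torso, with two large tilted boxes on spine_03 adding broad angled collision slabs.
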PhysicsAsset /BRCosmetics/Characters/Player/Female/Medium/Bodies/F_MED_Angler/Meshes/Parts/F_MED_Angler_BackpackOn_Physics
bodies {
 clavicle_l [{shapeType=Capsule, length=19.5902, radius=4.4862, center=<8 0 0>, rotation=<rot 90 0 0>}]
 clavicle_r [{shapeType=Capsule, length=19.5902, radius=4.4862, center=<-8 0 0>, rotation=<rot 90 0 0>}]
 dyn_hairMain_01 [{shapeType=Capsule, length=8.3965, radius=5.2597, center=<6.1784 0 0.0001>, rotation=<rot 90 0 0>}]
 dyn_hairMain_01_DRV [{shapeType=Capsule, length=0.505, radius=0.505, rotation=<rot 90 0 0>}]
 dyn_hairMain_01_DRV_Base [{shapeType=Capsule, length=0.505, radius=0.505, rotation=<rot 90 0 0>}]
 dyn_hairMain_02 [{shapeType=Capsule, length=8.6242, radius=6.5707, center=<7.0222 0 -0.0002>, rotation=<rot 90 0 0>}]
 dyn_hairMain_03 [{shapeType=Capsule, length=7.757, radius=4.5607, center=<4.2354 -1.127 0>, rotation=<rot 90 0 0>}]
 dyn_hairMain_04 [{shapeType=Capsule, length=10.1097, radius=3.9939, rotation=<rot 90 0 0>}]
 dyn_hairMain_05 [{shapeType=Capsule, length=10.1097, radius=3.9939, center=<0 -0.029 0>, rotation=<rot 90 0 0>}]
 dyn_hairMain_06 [{shapeType=Capsule, length=10.1097, radius=3.9939, center=<0 -0.029 0>, rotation=<rot 90 0 0>}]
 dyn_hairMain_07 [{shapeType=Capsule, length=10.1097, radius=3.9939, center=<0 -0.029 0>, rotation=<rot 90 0 0>}]
 pelvis [{shapeType=Capsule, length=14.2122, radius=11.182, center=<-1.9112 -3.606 0>}]
 spine_01 [{shapeType=Capsule, length=0.5217, radius=12.4097, rotation=<rot 90 0 0>}]
 spine_02 [{shapeType=Capsule, length=1.8127, radius=11.1187, rotation=<rot 90 0 0>}]
 spine_03 [
  {shapeType=Capsule, length=4.9561, radius=7.9754, rotation=<rot 90 0 0>}
  {shapeType=Box, extentX=95.8939, extentY=89.0945, extentZ=7.5, center=<0.0001 0.985 20.4673>, rotation=<rot 0 0 27.3824>}
  {shapeType=Box, extentX=102.6879, extentY=7.5, extentZ=64.9544, center=<0.0001 2.4908 -21.395>, rotation=<rot 0 0 58.591>}
 ]
 spine_04 [{shapeType=Capsule, length=9.4209, radius=7.8642, rotation=<rot 90 0 0>}]
 spine_05 [{shapeType=Capsule, length=13.3577, radius=11.0336, center=<0 4.6224 0>}]
}
constraints [
 {
  bone1Name=dyn_hairMain_01_DRV
  bone2Name=dyn_hairMain_01_DRV_Base
  swing1LimitDegrees=45
  swing1Motion=Free
  swing2LimitDegrees=45
  swing2Motion=Free
  twistLimitDegrees=45
  twistMotion=Free
 }
 {bone1Name=dyn_hairMain_01, bone2Name=dyn_hairMain_01_DRV, swing1LimitDegrees=35, swing2LimitDegrees=12, twistLimitDegrees=8}
 {bone1Name=dyn_hairMain_02, bone2Name=dyn_hairMain_01, swing1LimitDegrees=35, swing2LimitDegrees=12, twistLimitDegrees=8}
 {bone1Name=dyn_hairMain_03, bone2Name=dyn_hairMain_02, swing1LimitDegrees=20, swing2LimitDegrees=12, twistLimitDegrees=8}
 {bone1Name=dyn_hairMain_04, bone2Name=dyn_hairMain_03, swing1LimitDegrees=20, swing2LimitDegrees=12, twistLimitDegrees=8}
 {
  bone1Name=dyn_hairMain_01_DRV_Base
  bone2Name=spine_05
  swing1LimitDegrees=45
  swing1Motion=Free
  swing2LimitDegrees=45
  swing2Motion=Free
  twistLimitDegrees=45
  twistMotion=Free
 }
 {
  bone1Name=spine_01
  bone2Name=pelvis
  swing1LimitDegrees=45
  swing1Motion=Free
  swing2LimitDegrees=45
  swing2Motion=Free
  twistLimitDegrees=45
  twistMotion=Free
 }
 {
  bone1Name=spine_02
  bone2Name=spine_01
  swing1LimitDegrees=45
  swing1Motion=Free
  swing2LimitDegrees=45
  swing2Motion=Free
  twistLimitDegrees=45
  twistMotion=Free
 }
 {
  bone1Name=spine_03
  bone2Name=spine_02
  swing1LimitDegrees=45
  swing1Motion=Free
  swing2LimitDegrees=45
  swing2Motion=Free
  twistLimitDegrees=45
  twistMotion=Free
 }
 {
  bone1Name=spine_04
  bone2Name=spine_03
  swing1LimitDegrees=45
  swing1Motion=Free
  swing2LimitDegrees=45
  swing2Motion=Free
  twistLimitDegrees=45
  twistMotion=Free
 }
 {
  bone1Name=spine_05
  bone2Name=spine_04
  swing1LimitDegrees=45
  swing1Motion=Free
  swing2LimitDegrees=45
  swing2Motion=Free
  twistLimitDegrees=45
  twistMotion=Free
 }
 {bone1Name=dyn_hairMain_05, bone2Name=dyn_hairMain_04, swing1LimitDegrees=20, swing2LimitDegrees=12, twistLimitDegrees=8}
 {bone1Name=dyn_hairMain_06, bone2Name=dyn_hairMain_05, swing1LimitDegrees=20, swing2LimitDegrees=12, twistLimitDegrees=8}
 {bone1Name=dyn_hairMain_07, bone2Name=dyn_hairMain_06, swing1LimitDegrees=20, swing2LimitDegrees=12, twistLimitDegrees=8}
 {
  bone1Name=clavicle_l
  bone2Name=spine_05
  swing1LimitDegrees=45
  swing1Motion=Free
  swing2LimitDegrees=45
  swing2Motion=Free
  twistLimitDegrees=45
  twistMotion=Free
 }
 {
  bone1Name=clavicle_r
  bone2Name=spine_05
  swing1LimitDegrees=45
  swing1Motion=Free
  swing2LimitDegrees=45
  swing2Motion=Free
  twistLimitDegrees=45
  twistMotion=Free
 }
]
