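# Physics asset for the hair bangs of the F_MED_Angler cosmetic character body, defining the bodies and joints that let the bangs swing.
# Eight strands (A to D, left and right) are chains of three to five capsule bodies, each rooted to a single large head capsule.
# Angular limits keep twist tight throughout, with a wider second swing axis at each root than along the rest of the strand.
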PhysicsAsset /BRCosmetics/Characters/Player/Female/Medium/Bodies/F_MED_Angler/Meshes/Parts/F_MED_Angler_BangsPhysics
bodies {
 dyn_hairBangA_01_l [{shapeType=Capsule, length=2.5803, radius=1.0225, center=<2.2 0 0>, rotation=<rot 90 0 0>}]
 dyn_hairBangA_01_r [{shapeType=Capsule, length=4.1377, radius=1.0225, center=<2.4942 0.0001 0>, rotation=<rot 90 0 0>}]
 dyn_hairBangA_02_l [{shapeType=Capsule, length=2.5803, radius=1.0225, center=<1.9083 0 0>, rotation=<rot 90 0 0>}]
 dyn_hairBangA_02_r [{shapeType=Capsule, length=3.2033, radius=1.0225, center=<2.2 0 0>, rotation=<rot 90 0 0>}]
 dyn_hairBangA_03_l [{shapeType=Capsule, length=1.9012, radius=1.0225, center=<1.3925 0 0>, rotation=<rot 90 0 0>}]
 dyn_hairBangA_03_r [{shapeType=Capsule, length=2.8848, radius=1.0225, center=<2.2 0 0>, rotation=<rot 90 0 0>}]
 dyn_hairBangA_04_l [{shapeType=Capsule, length=2.2727, radius=0.7753, center=<1.3939 0 0>, rotation=<rot 90 0 0>}]
 dyn_hairBangA_04_r [{shapeType=Capsule, length=2.5803, radius=1.0225, center=<1.8812 0 0>, rotation=<rot 90 0 0>}]
 dyn_hairBangB_01_l [{shapeType=Capsule, length=3.9952, radius=1.0225, center=<2.467 0 0>, rotation=<rot 90 0 0>}]
 dyn_hairBangB_01_r [{shapeType=Capsule, length=5.0072, radius=1.0225, center=<2.72 0 0>, rotation=<rot 90 0 0>}]
 dyn_hairBangB_02_l [{shapeType=Capsule, length=3.3485, radius=1.0225, center=<2.2 0 0>, rotation=<rot 90 0 0>}]
 dyn_hairBangB_02_r [{shapeType=Capsule, length=4.1645, radius=1.0225, center=<2.2 0 0>, rotation=<rot 90 0 0>}]
 dyn_hairBangB_03_l [{shapeType=Capsule, length=3.4933, radius=1.0225, center=<2.2 0 0>, rotation=<rot 90 0 0>}]
 dyn_hairBangB_03_r [{shapeType=Capsule, length=3.9987, radius=1.0225, center=<2.2 0 0>, rotation=<rot 90 0 0>}]
 dyn_hairBangB_04_l [{shapeType=Capsule, length=2.0857, radius=0.7404, center=<1.2573 0.0001 0>, rotation=<rot 90 0 0>}]
 dyn_hairBangB_04_r [{shapeType=Capsule, length=3.5594, radius=1.0225, center=<2.2 0 0>, rotation=<rot 90 0 0>}]
 dyn_hairBangC_01_l [{shapeType=Capsule, length=5.0154, radius=1.0225, center=<2.5145 0 0>, rotation=<rot 90 0 0>}]
 dyn_hairBangC_01_r [{shapeType=Capsule, length=6.541, radius=1.0225, center=<2.9694 0 0>, rotation=<rot 90 0 0>}]
 dyn_hairBangC_02_l [{shapeType=Capsule, length=4.8813, radius=1.0225, center=<2.388 0 0>, rotation=<rot 90 0 0>}]
 dyn_hairBangC_02_r [{shapeType=Capsule, length=6.1336, radius=1.0225, center=<3.2871 0 0>, rotation=<rot 90 0 0>}]
 dyn_hairBangC_03_l [{shapeType=Capsule, length=2.8473, radius=1.0225, center=<1.5338 0 0>, rotation=<rot 90 0 0>}]
 dyn_hairBangC_03_r [{shapeType=Capsule, length=4.1332, radius=1.0225, center=<2.691 0 0>, rotation=<rot 90 0 0>}]
 dyn_hairBangC_04_r [{shapeType=Capsule, length=2.5025, radius=1.0225, center=<1.6218 0 0>, rotation=<rot 90 0 0>}]
 dyn_hairBangD_01_l [{shapeType=Capsule, length=7.9687, radius=1.0225, center=<3.2044 0 0>, rotation=<rot 90 0 0>}]
 dyn_hairBangD_02_l [{shapeType=Capsule, length=6.7643, radius=1.0225, center=<3.478 0 0>, rotation=<rot 90 0 0>}]
 dyn_hairBangD_03_l [{shapeType=Capsule, length=6.1064, radius=1.0225, center=<2.5764 0 0>, rotation=<rot 90 0 0>}]
 dyn_hairBangD_04_l [{shapeType=Capsule, length=2.9378, radius=0.9109, center=<1.5097 0 0>, rotation=<rot 90 0 0>}]
 dyn_hairBangD_05_l [{shapeType=Capsule, length=2.0397, radius=0.8021, center=<1.0338 0 0>, rotation=<rot 90 0 0>}]
 dyn_hairBang_D_01_r [{shapeType=Capsule, length=3.9952, radius=1.0225, center=<2.467 0 0>, rotation=<rot 90 0 0>}]
 dyn_hairBang_D_02_r [{shapeType=Capsule, length=3.3485, radius=1.0225, center=<2.2 0 0>, rotation=<rot 90 0 0>}]
 dyn_hairBang_D_03_r [{shapeType=Capsule, length=3.4933, radius=1.0225, center=<2.2 0 0>, rotation=<rot 90 0 0>}]
 dyn_hairBang_D_04_r [{shapeType=Capsule, length=2.0857, radius=0.7404, center=<1.2573 0.0001 0>, rotation=<rot 90 0 0>}]
 head [{shapeType=Capsule, length=10.045, radius=7.6178, rotation=<rot 90 9.4623 -21.0984>}]
}
constraints [
 {bone1Name=dyn_hairBangC_01_r, bone2Name=head, swing1LimitDegrees=35, swing2LimitDegrees=28, twistLimitDegrees=8}
 {bone1Name=dyn_hairBangC_02_r, bone2Name=dyn_hairBangC_01_r, swing1LimitDegrees=35, swing2LimitDegrees=12, twistLimitDegrees=8}
 {bone1Name=dyn_hairBangC_03_r, bone2Name=dyn_hairBangC_02_r, swing1LimitDegrees=35, swing2LimitDegrees=12, twistLimitDegrees=8}
 {bone1Name=dyn_hairBangC_04_r, bone2Name=dyn_hairBangC_03_r, swing1LimitDegrees=35, swing2LimitDegrees=12, twistLimitDegrees=8}
 {bone1Name=dyn_hairBangB_01_r, bone2Name=head, swing1LimitDegrees=35, swing2LimitDegrees=28, twistLimitDegrees=8}
 {bone1Name=dyn_hairBangB_02_r, bone2Name=dyn_hairBangB_01_r, swing1LimitDegrees=35, swing2LimitDegrees=12, twistLimitDegrees=8}
 {bone1Name=dyn_hairBangB_03_r, bone2Name=dyn_hairBangB_02_r, swing1LimitDegrees=35, swing2LimitDegrees=12, twistLimitDegrees=8}
 {bone1Name=dyn_hairBangB_04_r, bone2Name=dyn_hairBangB_03_r, swing1LimitDegrees=35, swing2LimitDegrees=12, twistLimitDegrees=8}
 {bone1Name=dyn_hairBangA_01_r, bone2Name=head, swing1LimitDegrees=35, swing2LimitDegrees=40, twistLimitDegrees=8}
 {bone1Name=dyn_hairBangA_02_r, bone2Name=dyn_hairBangA_01_r, swing1LimitDegrees=35, swing2LimitDegrees=12, twistLimitDegrees=8}
 {bone1Name=dyn_hairBangA_03_r, bone2Name=dyn_hairBangA_02_r, swing1LimitDegrees=35, swing2LimitDegrees=12, twistLimitDegrees=8}
 {bone1Name=dyn_hairBangA_04_r, bone2Name=dyn_hairBangA_03_r, swing1LimitDegrees=35, swing2LimitDegrees=12, twistLimitDegrees=8}
 {bone1Name=dyn_hairBangA_01_l, bone2Name=head, swing1LimitDegrees=35, swing2LimitDegrees=40, twistLimitDegrees=8}
 {bone1Name=dyn_hairBangA_02_l, bone2Name=dyn_hairBangA_01_l, swing1LimitDegrees=35, swing2LimitDegrees=12, twistLimitDegrees=8}
 {bone1Name=dyn_hairBangA_03_l, bone2Name=dyn_hairBangA_02_l, swing1LimitDegrees=35, swing2LimitDegrees=12, twistLimitDegrees=8}
 {bone1Name=dyn_hairBangA_04_l, bone2Name=dyn_hairBangA_03_l, swing1LimitDegrees=35, swing2LimitDegrees=12, twistLimitDegrees=8}
 {bone1Name=dyn_hairBangB_01_l, bone2Name=head, swing1LimitDegrees=35, swing2LimitDegrees=40, twistLimitDegrees=8}
 {bone1Name=dyn_hairBangB_02_l, bone2Name=dyn_hairBangB_01_l, swing1LimitDegrees=35, swing2LimitDegrees=12, twistLimitDegrees=8}
 {bone1Name=dyn_hairBangB_03_l, bone2Name=dyn_hairBangB_02_l, swing1LimitDegrees=35, swing2LimitDegrees=12, twistLimitDegrees=8}
 {bone1Name=dyn_hairBangB_04_l, bone2Name=dyn_hairBangB_03_l, swing1LimitDegrees=35, swing2LimitDegrees=12, twistLimitDegrees=8}
 {bone1Name=dyn_hairBangC_01_l, bone2Name=head, swing1LimitDegrees=35, swing2LimitDegrees=40, twistLimitDegrees=8}
 {bone1Name=dyn_hairBangC_02_l, bone2Name=dyn_hairBangC_01_l, swing1LimitDegrees=35, swing2LimitDegrees=12, twistLimitDegrees=8}
 {bone1Name=dyn_hairBangC_03_l, bone2Name=dyn_hairBangC_02_l, swing1LimitDegrees=35, swing2LimitDegrees=12, twistLimitDegrees=8}
 {bone1Name=dyn_hairBangD_01_l, bone2Name=head, swing1LimitDegrees=35, swing2LimitDegrees=28, twistLimitDegrees=8}
 {bone1Name=dyn_hairBangD_02_l, bone2Name=dyn_hairBangD_01_l, swing1LimitDegrees=35, swing2LimitDegrees=12, twistLimitDegrees=8}
 {bone1Name=dyn_hairBangD_03_l, bone2Name=dyn_hairBangD_02_l, swing1LimitDegrees=35, swing2LimitDegrees=12, twistLimitDegrees=8}
 {bone1Name=dyn_hairBangD_04_l, bone2Name=dyn_hairBangD_03_l, swing1LimitDegrees=35, swing2LimitDegrees=12, twistLimitDegrees=8}
 {bone1Name=dyn_hairBangD_05_l, bone2Name=dyn_hairBangD_04_l, swing1LimitDegrees=35, swing2LimitDegrees=12, twistLimitDegrees=8}
 {bone1Name=dyn_hairBang_D_01_r, bone2Name=head, swing1LimitDegrees=35, swing2LimitDegrees=40, twistLimitDegrees=8}
 {bone1Name=dyn_hairBang_D_02_r, bone2Name=dyn_hairBang_D_01_r, swing1LimitDegrees=35, swing2LimitDegrees=12, twistLimitDegrees=8}
 {bone1Name=dyn_hairBang_D_03_r, bone2Name=dyn_hairBang_D_02_r, swing1LimitDegrees=35, swing2LimitDegrees=12, twistLimitDegrees=8}
 {bone1Name=dyn_hairBang_D_04_r, bone2Name=dyn_hairBang_D_03_r, swing1LimitDegrees=35, swing2LimitDegrees=12, twistLimitDegrees=8}
]
