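# Control Rig for the Angler female body that procedurally animates the long main hair and its strands.
# Chain harmonics adds pendulum and wave motion to the main chain.
# Strand bones mostly follow main-chain local rotations, using per-bone slerp and accumulated lerp for varied lag.
# Head rotation relative to spine_05 then drives corrective hair-root offsets weighted by backblingFloat; per-strand harmonics and some additive nodes sit unconnected.
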ControlRigBlueprint /BRCosmetics/Characters/Player/Female/Medium/Bodies/F_MED_Angler/Meshes/Parts/F_MED_Angler_CtrlRig/F_MED_Angler_CtrlRig
graphs [
 RigVMFunctionLibrary {links [], nodes []}
 RigVMModel {nodes [
   /Script/ControlRig.RigUnit_ChainHarmonics(
    ChainRoot=dyn_hairStrandA_01
    DrawWorldOffset=(Rotation=(X=0,Y=0,Z=0,W=1),Translation=(X=0,Y=0,Z=0),Scale3D=(X=1,Y=1,Z=1))
    ExecutePin=()
    Pendulum=(bEnabled=True,PendulumStiffness=2,PendulumGravity=(X=0,Y=0,Z=0),PendulumBlend=0.75,PendulumDrag=0.98,PendulumMaximum=0.1,UnwindAxis=(X=0,Y=1,Z=0),UnwindMinimum=0.2,UnwindMaximum=0.05)
    Reach=(bEnabled=false,ReachTarget=(X=-10,Y=-100,Z=0),ReachAxis=(X=0,Y=0,Z=0),ReachMinimum=0)
    Speed=(X=1,Y=0.8,Z=1)
    Wave=(bEnabled=True,WaveFrequency=(X=0,Y=10,Z=0),WaveAmplitude=(X=0,Y=0.8,Z=0),WaveOffset=(X=0,Y=0,Z=0),WaveNoise=(X=0,Y=0,Z=0),WaveMaximum=1.25,WaveEase=QuarticEaseInOut)
    WaveCurve=(EditorCurveData=(Keys=((Time=-0.0782,Value=0.8776),(Time=1,Value=1)),DefaultValue=340282346638528859811704183484516925440,PreInfinityExtrap=RCCE_Constant,PostInfinityExtrap=RCCE_Constant))
    bDrawDebug=True
   ) {is=struct RigVMUnitNode}
   > &RigUnit_ChainHarmonics(
    ChainRoot=dyn_hairStrandB_01
    DrawWorldOffset=(Rotation=(X=0,Y=0,Z=0,W=1),Translation=(X=0,Y=0,Z=0),Scale3D=(X=1,Y=1,Z=1))
    Pendulum=(bEnabled=True,PendulumStiffness=2,PendulumGravity=(X=0,Y=0,Z=0),PendulumBlend=0.75,PendulumDrag=0.98,PendulumMaximum=0.1,UnwindAxis=(X=0,Y=1,Z=0),UnwindMinimum=0.2,UnwindMaximum=0.05)
    Reach=(bEnabled=false,ReachTarget=(X=-10,Y=-100,Z=0),ReachAxis=(X=0,Y=0,Z=0),ReachMinimum=0)
    Speed=(X=1,Y=0.75,Z=1)
    Wave=(bEnabled=True,WaveFrequency=(X=0,Y=10,Z=0),WaveAmplitude=(X=0,Y=0.8,Z=0),WaveOffset=(X=0,Y=0,Z=0),WaveNoise=(X=0,Y=0,Z=0),WaveMaximum=1.25,WaveEase=QuarticEaseInOut)
    WaveCurve=(EditorCurveData=(Keys=((Time=-0.0782,Value=0.8776),(Time=1,Value=1)),DefaultValue=340282346638528859811704183484516925440,PreInfinityExtrap=RCCE_Constant,PostInfinityExtrap=RCCE_Constant))
    bDrawDebug=True
   )
   > &RigUnit_ChainHarmonics(
    ChainRoot=dyn_hairStrandC_01
    DrawWorldOffset=(Rotation=(X=0,Y=0,Z=0,W=1),Translation=(X=0,Y=0,Z=0),Scale3D=(X=1,Y=1,Z=1))
    Pendulum=(bEnabled=True,PendulumStiffness=2,PendulumGravity=(X=0,Y=0,Z=0),PendulumBlend=0.75,PendulumDrag=0.98,PendulumMaximum=0.1,UnwindAxis=(X=0,Y=1,Z=0),UnwindMinimum=0.2,UnwindMaximum=0.05)
    Reach=(bEnabled=false,ReachTarget=(X=-10,Y=-100,Z=0),ReachAxis=(X=0,Y=0,Z=0),ReachMinimum=0)
    Speed=(X=1,Y=0.5,Z=1)
    Wave=(bEnabled=True,WaveFrequency=(X=0,Y=10,Z=0),WaveAmplitude=(X=0,Y=0.8,Z=0),WaveOffset=(X=0,Y=0,Z=0),WaveNoise=(X=0,Y=0,Z=0),WaveMaximum=1.25,WaveEase=QuarticEaseInOut)
    WaveCurve=(EditorCurveData=(Keys=((Time=-0.0782,Value=0.8776),(Time=1,Value=1)),DefaultValue=340282346638528859811704183484516925440,PreInfinityExtrap=RCCE_Constant,PostInfinityExtrap=RCCE_Constant))
    bDrawDebug=True
   )
   > &RigUnit_ChainHarmonics(
    ChainRoot=dyn_hairStrandD_01
    DrawWorldOffset=(Rotation=(X=0,Y=0,Z=0,W=1),Translation=(X=0,Y=0,Z=0),Scale3D=(X=1,Y=1,Z=1))
    Pendulum=(bEnabled=True,PendulumStiffness=2,PendulumGravity=(X=0,Y=0,Z=0),PendulumBlend=0.75,PendulumDrag=0.98,PendulumMaximum=0.1,UnwindAxis=(X=0,Y=1,Z=0),UnwindMinimum=0.2,UnwindMaximum=0.05)
    Reach=(bEnabled=false,ReachTarget=(X=-10,Y=-100,Z=0),ReachAxis=(X=0,Y=0,Z=0),ReachMinimum=0)
    Speed=(X=1,Y=0.954,Z=1)
    Wave=(bEnabled=True,WaveFrequency=(X=0,Y=10,Z=0),WaveAmplitude=(X=0,Y=0.8,Z=0),WaveOffset=(X=0,Y=0,Z=0),WaveNoise=(X=0,Y=0,Z=0),WaveMaximum=1.25,WaveEase=QuarticEaseInOut)
    WaveCurve=(EditorCurveData=(Keys=((Time=-0.0782,Value=0.8776),(Time=1,Value=1)),DefaultValue=340282346638528859811704183484516925440,PreInfinityExtrap=RCCE_Constant,PostInfinityExtrap=RCCE_Constant))
    bDrawDebug=True
   )
   > &RigUnit_ChainHarmonics(
    ChainRoot=dyn_hairStrandE_01
    DrawWorldOffset=(Rotation=(X=0,Y=0,Z=0,W=1),Translation=(X=0,Y=0,Z=0),Scale3D=(X=1,Y=1,Z=1))
    Pendulum=(bEnabled=True,PendulumStiffness=2,PendulumGravity=(X=0,Y=0,Z=0),PendulumBlend=0.75,PendulumDrag=0.98,PendulumMaximum=0.1,UnwindAxis=(X=0,Y=1,Z=0),UnwindMinimum=0.2,UnwindMaximum=0.05)
    Reach=(bEnabled=false,ReachTarget=(X=-10,Y=-100,Z=0),ReachAxis=(X=0,Y=0,Z=0),ReachMinimum=0)
    Speed=(X=1,Y=0.85,Z=1)
    Wave=(bEnabled=True,WaveFrequency=(X=0,Y=10,Z=0),WaveAmplitude=(X=0,Y=0.8,Z=0),WaveOffset=(X=0,Y=0,Z=0),WaveNoise=(X=0,Y=0,Z=0),WaveMaximum=1.25,WaveEase=QuarticEaseInOut)
    WaveCurve=(EditorCurveData=(Keys=((Time=-0.0782,Value=0.8776),(Time=1,Value=1)),DefaultValue=340282346638528859811704183484516925440,PreInfinityExtrap=RCCE_Constant,PostInfinityExtrap=RCCE_Constant))
    bDrawDebug=True
   )
   n1 = &RigUnit_GetBoneTransform(Bone=dyn_hairMain_04, Space=LocalSpace)
   /Script/ControlRig.RigUnit_ModifyBoneTransforms(
    BoneToModify=((Bone="dyn_hairStrandD_01",Transform=(Rotation=(X=0,Y=0,Z=0,W=1),Translation=(X=0,Y=0,Z=0),Scale3D=(X=1,Y=1,Z=1))))
    ExecutePin=()
    Mode=AdditiveLocal
    Weight=1
    WeightMaximum=1
    BoneToModify.0.Transform.Rotation=/Script/RigVM.RigVMFunction_AccumulateQuatLerp(
     Blend=0.75
     bIntegrateDeltaTime=true
     TargetValue=@n1.Transform.Rotation
     InitialValue=&RigUnit_GetBoneTransform(Bone=dyn_hairStrandD_01, Space=LocalSpace).Transform.Rotation
    ) {is=struct RigVMUnitNode, template="AccumulateLerp::Execute(in TargetValue,in InitialValue,in Blend,in bIntegrateDeltaTime,out Result)"}.Result
   ) {is=struct RigVMUnitNode}
   > &RigUnit_ModifyBoneTransforms(
    BoneToModify=((Bone="dyn_hairStrandA_01",Transform=(Rotation=(X=0,Y=0,Z=0,W=1),Translation=(X=0,Y=0,Z=0),Scale3D=(X=1,Y=1,Z=1))))
    Mode=AdditiveLocal
    Weight=1
    WeightMaximum=1
    BoneToModify.0.Transform.Rotation=&RigVMFunction_AccumulateQuatLerp(
     Blend=0.25
     bIntegrateDeltaTime=true
     TargetValue=@n1.Transform.Rotation
     InitialValue=&RigUnit_GetBoneTransform(Bone=dyn_hairStrandA_01, Space=LocalSpace).Transform.Rotation
    ) {template="AccumulateLerp::Execute(in TargetValue,in InitialValue,in Blend,in bIntegrateDeltaTime,out Result)"}.Result
   )
   > &RigUnit_ModifyBoneTransforms(
    BoneToModify=((Bone="dyn_hairStrandB_01",Transform=(Rotation=(X=0,Y=0,Z=0,W=1),Translation=(X=0,Y=0,Z=0),Scale3D=(X=1,Y=1,Z=1))))
    Mode=AdditiveLocal
    Weight=1
    WeightMaximum=1
    BoneToModify.0.Transform.Rotation=&RigVMFunction_AccumulateQuatLerp(
     Blend=0.5
     bIntegrateDeltaTime=true
     TargetValue=@n1.Transform.Rotation
     InitialValue=&RigUnit_GetBoneTransform(Bone=dyn_hairStrandB_01, Space=LocalSpace).Transform.Rotation
    ) {template="AccumulateLerp::Execute(in TargetValue,in InitialValue,in Blend,in bIntegrateDeltaTime,out Result)"}.Result
   )
   RigUnit_BeginExecution = &RigUnit_BeginExecution()
   &RigUnit_ChainHarmonics(
    ChainRoot=dyn_hairMain_01_DRV
    DrawWorldOffset=(Rotation=(X=0,Y=0,Z=0,W=1),Translation=(X=0,Y=0,Z=0),Scale3D=(X=1,Y=1,Z=1))
    Pendulum=(bEnabled=True,PendulumStiffness=2,PendulumGravity=(X=0,Y=0,Z=0),PendulumBlend=0.75,PendulumDrag=0.98,PendulumMaximum=0.1,UnwindAxis=(X=0,Y=1,Z=0),UnwindMinimum=0.2,UnwindMaximum=0.05)
    Reach=(bEnabled=false,ReachTarget=(X=-10,Y=-100,Z=0),ReachAxis=(X=0,Y=0,Z=0),ReachMinimum=0)
    Speed=(X=1,Y=1,Z=1)
    Wave=(bEnabled=True,WaveFrequency=(X=0,Y=10,Z=0),WaveAmplitude=(X=0,Y=0.8,Z=0),WaveOffset=(X=0,Y=0,Z=0),WaveNoise=(X=0,Y=0,Z=0),WaveMaximum=1.25,WaveEase=Linear)
    WaveCurve=(EditorCurveData=(Keys=((Value=0.25),(Time=0.329,Value=0.4542),(Time=0.6847,Value=0.9989)),DefaultValue=340282346638528859811704183484516925440,PreInfinityExtrap=RCCE_Constant,PostInfinityExtrap=RCCE_Constant))
    bDrawDebug=True
    Reach.ReachTarget=&RigUnit_GetBoneTransform(Bone=head).Transform.Translation
    ExecutePin=@RigUnit_BeginExecution.ExecutePin
   )
   n2 = &RigUnit_GetBoneTransform(Bone=dyn_hairMain_04, Space=LocalSpace)
   > /Script/ControlRig.RigUnit_SetBoneRotation(
    Bone=dyn_hairStrandA_01
    Weight=1
    bPropagateToChildren=True
    Rotation=&RigVMFunction_AccumulateQuatLerp(
     Blend=0.5
     InitialValue=(X=0,Y=0,Z=0,W=1)
     TargetValue=/Script/RigVM.RigVMFunction_MathQuaternionSlerp(A=(X=0,Y=0,Z=0,W=1), T=0.725, B=@n2.Transform.Rotation) {is=struct RigVMUnitNode, template="Interpolate::Execute(in A,in B,in T,out Result)"}.Result
    ) {template="AccumulateLerp::Execute(in TargetValue,in InitialValue,in Blend,in bIntegrateDeltaTime,out Result)"}.Result
   ) {is=struct RigVMUnitNode}
   > &RigUnit_SetBoneRotation(
    Bone=dyn_hairStrandB_01
    Weight=1
    bPropagateToChildren=True
    Rotation=&RigVMFunction_AccumulateQuatLerp(
     Blend=0.35
     InitialValue=(X=0,Y=0,Z=0,W=1)
     TargetValue=&RigVMFunction_MathQuaternionSlerp(A=(X=0,Y=0,Z=0,W=1), T=0.8, B=@n2.Transform.Rotation) {template="Interpolate::Execute(in A,in B,in T,out Result)"}.Result
    ) {template="AccumulateLerp::Execute(in TargetValue,in InitialValue,in Blend,in bIntegrateDeltaTime,out Result)"}.Result
   )
   RigVMStructNode_0 = &RigVMFunction_AccumulateQuatLerp(
    Blend=0.2
    InitialValue=(X=0,Y=0,Z=0,W=1)
    TargetValue=&RigVMFunction_MathQuaternionSlerp(A=(X=0,Y=0,Z=0,W=1), T=0.675, B=@n2.Transform.Rotation) {template="Interpolate::Execute(in A,in B,in T,out Result)"}.Result
   ) {template="AccumulateLerp::Execute(in TargetValue,in InitialValue,in Blend,in bIntegrateDeltaTime,out Result)"}
   > &RigUnit_SetBoneRotation(Bone=dyn_hairStrandC_01, Weight=1, bPropagateToChildren=True, Rotation=@RigVMStructNode_0.Result)
   > &RigUnit_SetBoneRotation(
    Bone=dyn_hairStrandD_01
    Weight=1
    bPropagateToChildren=True
    Rotation=&RigVMFunction_AccumulateQuatLerp(
     Blend=0.2
     InitialValue=(X=0,Y=0,Z=0,W=1)
     TargetValue=&RigVMFunction_MathQuaternionSlerp(A=(X=0,Y=0,Z=0,W=1), T=0.65, B=@n2.Transform.Rotation) {template="Interpolate::Execute(in A,in B,in T,out Result)"}.Result
    ) {template="AccumulateLerp::Execute(in TargetValue,in InitialValue,in Blend,in bIntegrateDeltaTime,out Result)"}.Result
   )
   > &RigUnit_SetBoneRotation(
    Bone=dyn_hairStrandE_01
    Weight=1
    bPropagateToChildren=True
    Rotation=&RigVMFunction_AccumulateQuatLerp(
     Blend=0.4
     InitialValue=(X=0,Y=0,Z=0,W=1)
     TargetValue=&RigVMFunction_MathQuaternionSlerp(A=(X=0,Y=0,Z=0,W=1), T=0.5, B=@n2.Transform.Rotation) {template="Interpolate::Execute(in A,in B,in T,out Result)"}.Result
    ) {template="AccumulateLerp::Execute(in TargetValue,in InitialValue,in Blend,in bIntegrateDeltaTime,out Result)"}.Result
   )
   > &RigUnit_SetBoneRotation(
    Bone=dyn_hairStrandG_01
    Weight=1
    bPropagateToChildren=True
    Rotation=&RigVMFunction_AccumulateQuatLerp(
     Blend=0.5
     InitialValue=(X=0,Y=0,Z=0,W=1)
     TargetValue=&RigVMFunction_MathQuaternionSlerp(A=(X=0,Y=0,Z=0,W=1), T=0.725, B=@n2.Transform.Rotation) {template="Interpolate::Execute(in A,in B,in T,out Result)"}.Result
    ) {template="AccumulateLerp::Execute(in TargetValue,in InitialValue,in Blend,in bIntegrateDeltaTime,out Result)"}.Result
   )
   > &RigUnit_SetBoneRotation(
    Bone=dyn_hairStrandH_01
    Weight=1
    bPropagateToChildren=True
    Rotation=&RigVMFunction_AccumulateQuatLerp(
     Blend=0.35
     InitialValue=(X=0,Y=0,Z=0,W=1)
     TargetValue=&RigVMFunction_MathQuaternionSlerp(A=(X=0,Y=0,Z=0,W=1), T=0.8, B=@n2.Transform.Rotation) {template="Interpolate::Execute(in A,in B,in T,out Result)"}.Result
    ) {template="AccumulateLerp::Execute(in TargetValue,in InitialValue,in Blend,in bIntegrateDeltaTime,out Result)"}.Result
   )
   RigVMStructNode_0_2 = &RigVMFunction_AccumulateQuatLerp(
    Blend=0.2
    InitialValue=(X=0,Y=0,Z=0,W=1)
    TargetValue=&RigVMFunction_MathQuaternionSlerp(A=(X=0,Y=0,Z=0,W=1), T=0.675, B=@n2.Transform.Rotation) {template="Interpolate::Execute(in A,in B,in T,out Result)"}.Result
   ) {template="AccumulateLerp::Execute(in TargetValue,in InitialValue,in Blend,in bIntegrateDeltaTime,out Result)"}
   > &RigUnit_SetBoneRotation(Bone=dyn_hairStrandJ_01, Weight=1, bPropagateToChildren=True, Rotation=@RigVMStructNode_0_2.Result)
   > &RigUnit_SetBoneRotation(
    Bone=dyn_hairStrandK_01
    Weight=1
    bPropagateToChildren=True
    Rotation=&RigVMFunction_AccumulateQuatLerp(
     Blend=0.2
     InitialValue=(X=0,Y=0,Z=0,W=1)
     TargetValue=&RigVMFunction_MathQuaternionSlerp(A=(X=0,Y=0,Z=0,W=1), T=0.65, B=@n2.Transform.Rotation) {template="Interpolate::Execute(in A,in B,in T,out Result)"}.Result
    ) {template="AccumulateLerp::Execute(in TargetValue,in InitialValue,in Blend,in bIntegrateDeltaTime,out Result)"}.Result
   )
   > &RigUnit_SetBoneRotation(
    Bone=dyn_hairStrandL_01
    Weight=1
    bPropagateToChildren=True
    Rotation=&RigVMFunction_AccumulateQuatLerp(
     Blend=0.4
     InitialValue=(X=0,Y=0,Z=0,W=1)
     TargetValue=&RigVMFunction_MathQuaternionSlerp(A=(X=0,Y=0,Z=0,W=1), T=0.5, B=@n2.Transform.Rotation) {template="Interpolate::Execute(in A,in B,in T,out Result)"}.Result
    ) {template="AccumulateLerp::Execute(in TargetValue,in InitialValue,in Blend,in bIntegrateDeltaTime,out Result)"}.Result
   )
   > &RigUnit_SetBoneRotation(
    Bone=dyn_hairStrandF_01
    Weight=1
    bPropagateToChildren=True
    Rotation=&RigVMFunction_AccumulateQuatLerp(
     Blend=0.4
     InitialValue=(X=0,Y=0,Z=0,W=1)
     TargetValue=&RigVMFunction_MathQuaternionSlerp(A=(X=0,Y=0,Z=0,W=1), T=0.5, B=@n2.Transform.Rotation) {template="Interpolate::Execute(in A,in B,in T,out Result)"}.Result
    ) {template="AccumulateLerp::Execute(in TargetValue,in InitialValue,in Blend,in bIntegrateDeltaTime,out Result)"}.Result
   )
   n3 = &RigUnit_GetBoneTransform(Bone=dyn_hairMain_05, Space=LocalSpace)
   > &RigUnit_SetBoneRotation(
    Bone=dyn_hairStrandA_02
    Weight=1
    bPropagateToChildren=True
    Rotation=&RigVMFunction_AccumulateQuatLerp(
     Blend=0.5
     InitialValue=(X=0,Y=0,Z=0,W=1)
     TargetValue=&RigVMFunction_MathQuaternionSlerp(A=(X=0,Y=0,Z=0,W=1), T=0.725, B=@n3.Transform.Rotation) {template="Interpolate::Execute(in A,in B,in T,out Result)"}.Result
    ) {template="AccumulateLerp::Execute(in TargetValue,in InitialValue,in Blend,in bIntegrateDeltaTime,out Result)"}.Result
   )
   > &RigUnit_SetBoneRotation(
    Bone=dyn_hairStrandB_02
    Weight=1
    bPropagateToChildren=True
    Rotation=&RigVMFunction_AccumulateQuatLerp(
     Blend=0.35
     InitialValue=(X=0,Y=0,Z=0,W=1)
     TargetValue=&RigVMFunction_MathQuaternionSlerp(A=(X=0,Y=0,Z=0,W=1), T=0.8, B=@n3.Transform.Rotation) {template="Interpolate::Execute(in A,in B,in T,out Result)"}.Result
    ) {template="AccumulateLerp::Execute(in TargetValue,in InitialValue,in Blend,in bIntegrateDeltaTime,out Result)"}.Result
   )
   RigVMStructNode_0_1 = &RigVMFunction_AccumulateQuatLerp(
    Blend=0.2
    InitialValue=(X=0,Y=0,Z=0,W=1)
    TargetValue=&RigVMFunction_MathQuaternionSlerp(A=(X=0,Y=0,Z=0,W=1), T=0.675, B=@n3.Transform.Rotation) {template="Interpolate::Execute(in A,in B,in T,out Result)"}.Result
   ) {template="AccumulateLerp::Execute(in TargetValue,in InitialValue,in Blend,in bIntegrateDeltaTime,out Result)"}
   > &RigUnit_SetBoneRotation(Bone=dyn_hairStrandC_02, Weight=1, bPropagateToChildren=True, Rotation=@RigVMStructNode_0_1.Result)
   > &RigUnit_SetBoneRotation(
    Bone=dyn_hairStrandD_02
    Weight=1
    bPropagateToChildren=True
    Rotation=&RigVMFunction_AccumulateQuatLerp(
     Blend=0.2
     InitialValue=(X=0,Y=0,Z=0,W=1)
     TargetValue=&RigVMFunction_MathQuaternionSlerp(A=(X=0,Y=0,Z=0,W=1), T=0.65, B=@n3.Transform.Rotation) {template="Interpolate::Execute(in A,in B,in T,out Result)"}.Result
    ) {template="AccumulateLerp::Execute(in TargetValue,in InitialValue,in Blend,in bIntegrateDeltaTime,out Result)"}.Result
   )
   > &RigUnit_SetBoneRotation(
    Bone=dyn_hairStrandE_02
    Weight=1
    bPropagateToChildren=True
    Rotation=&RigVMFunction_AccumulateQuatLerp(
     Blend=0.4
     InitialValue=(X=0,Y=0,Z=0,W=1)
     TargetValue=&RigVMFunction_MathQuaternionSlerp(A=(X=0,Y=0,Z=0,W=1), T=0.5, B=@n3.Transform.Rotation) {template="Interpolate::Execute(in A,in B,in T,out Result)"}.Result
    ) {template="AccumulateLerp::Execute(in TargetValue,in InitialValue,in Blend,in bIntegrateDeltaTime,out Result)"}.Result
   )
   > &RigUnit_SetBoneRotation(
    Bone=dyn_hairStrandG_02
    Weight=1
    bPropagateToChildren=True
    Rotation=&RigVMFunction_AccumulateQuatLerp(
     Blend=0.5
     InitialValue=(X=0,Y=0,Z=0,W=1)
     TargetValue=&RigVMFunction_MathQuaternionSlerp(A=(X=0,Y=0,Z=0,W=1), T=0.725, B=@n3.Transform.Rotation) {template="Interpolate::Execute(in A,in B,in T,out Result)"}.Result
    ) {template="AccumulateLerp::Execute(in TargetValue,in InitialValue,in Blend,in bIntegrateDeltaTime,out Result)"}.Result
   )
   > &RigUnit_SetBoneRotation(
    Bone=dyn_hairStrandH_02
    Weight=1
    bPropagateToChildren=True
    Rotation=&RigVMFunction_AccumulateQuatLerp(
     Blend=0.35
     InitialValue=(X=0,Y=0,Z=0,W=1)
     TargetValue=&RigVMFunction_MathQuaternionSlerp(A=(X=0,Y=0,Z=0,W=1), T=0.8, B=@n3.Transform.Rotation) {template="Interpolate::Execute(in A,in B,in T,out Result)"}.Result
    ) {template="AccumulateLerp::Execute(in TargetValue,in InitialValue,in Blend,in bIntegrateDeltaTime,out Result)"}.Result
   )
   RigVMStructNode_0_1_2 = &RigVMFunction_AccumulateQuatLerp(
    Blend=0.2
    InitialValue=(X=0,Y=0,Z=0,W=1)
    TargetValue=&RigVMFunction_MathQuaternionSlerp(A=(X=0,Y=0,Z=0,W=1), T=0.675, B=@n3.Transform.Rotation) {template="Interpolate::Execute(in A,in B,in T,out Result)"}.Result
   ) {template="AccumulateLerp::Execute(in TargetValue,in InitialValue,in Blend,in bIntegrateDeltaTime,out Result)"}
   > &RigUnit_SetBoneRotation(Bone=dyn_hairStrandJ_02, Weight=1, bPropagateToChildren=True, Rotation=@RigVMStructNode_0_1_2.Result)
   > &RigUnit_SetBoneRotation(
    Bone=dyn_hairStrandK_02
    Weight=1
    bPropagateToChildren=True
    Rotation=&RigVMFunction_AccumulateQuatLerp(
     Blend=0.2
     InitialValue=(X=0,Y=0,Z=0,W=1)
     TargetValue=&RigVMFunction_MathQuaternionSlerp(A=(X=0,Y=0,Z=0,W=1), T=0.65, B=@n3.Transform.Rotation) {template="Interpolate::Execute(in A,in B,in T,out Result)"}.Result
    ) {template="AccumulateLerp::Execute(in TargetValue,in InitialValue,in Blend,in bIntegrateDeltaTime,out Result)"}.Result
   )
   > &RigUnit_SetBoneRotation(
    Bone=dyn_hairStrandL_02
    Weight=1
    bPropagateToChildren=True
    Rotation=&RigVMFunction_AccumulateQuatLerp(
     Blend=0.4
     InitialValue=(X=0,Y=0,Z=0,W=1)
     TargetValue=&RigVMFunction_MathQuaternionSlerp(A=(X=0,Y=0,Z=0,W=1), T=0.5, B=@n3.Transform.Rotation) {template="Interpolate::Execute(in A,in B,in T,out Result)"}.Result
    ) {template="AccumulateLerp::Execute(in TargetValue,in InitialValue,in Blend,in bIntegrateDeltaTime,out Result)"}.Result
   )
   > &RigUnit_SetBoneRotation(
    Bone=dyn_hairStrandF_02
    Weight=1
    bPropagateToChildren=True
    Rotation=&RigVMFunction_AccumulateQuatLerp(
     Blend=0.4
     InitialValue=(X=0,Y=0,Z=0,W=1)
     TargetValue=&RigVMFunction_MathQuaternionSlerp(A=(X=0,Y=0,Z=0,W=1), T=0.5, B=@n3.Transform.Rotation) {template="Interpolate::Execute(in A,in B,in T,out Result)"}.Result
    ) {template="AccumulateLerp::Execute(in TargetValue,in InitialValue,in Blend,in bIntegrateDeltaTime,out Result)"}.Result
   )
   n4 = &RigUnit_GetBoneTransform(Bone=dyn_hairMain_06, Space=LocalSpace)
   > &RigUnit_SetBoneRotation(
    Bone=dyn_hairStrandD_03
    Weight=1
    bPropagateToChildren=True
    Rotation=&RigVMFunction_AccumulateQuatLerp(
     Blend=0.5
     InitialValue=(X=0,Y=0,Z=0,W=1)
     TargetValue=&RigVMFunction_MathQuaternionSlerp(A=(X=0,Y=0,Z=0,W=1), T=0.725, B=@n4.Transform.Rotation) {template="Interpolate::Execute(in A,in B,in T,out Result)"}.Result
    ) {template="AccumulateLerp::Execute(in TargetValue,in InitialValue,in Blend,in bIntegrateDeltaTime,out Result)"}.Result
   )
   > &RigUnit_SetBoneRotation(
    Bone=dyn_hairStrandC_03
    Weight=1
    bPropagateToChildren=True
    Rotation=&RigVMFunction_AccumulateQuatLerp(
     Blend=0.35
     InitialValue=(X=0,Y=0,Z=0,W=1)
     TargetValue=&RigVMFunction_MathQuaternionSlerp(A=(X=0,Y=0,Z=0,W=1), T=0.8, B=@n4.Transform.Rotation) {template="Interpolate::Execute(in A,in B,in T,out Result)"}.Result
    ) {template="AccumulateLerp::Execute(in TargetValue,in InitialValue,in Blend,in bIntegrateDeltaTime,out Result)"}.Result
   )
   RigVMStructNode_0_1_1 = &RigVMFunction_AccumulateQuatLerp(
    Blend=0.2
    InitialValue=(X=0,Y=0,Z=0,W=1)
    TargetValue=&RigVMFunction_MathQuaternionSlerp(A=(X=0,Y=0,Z=0,W=1), T=0.675, B=@n4.Transform.Rotation) {template="Interpolate::Execute(in A,in B,in T,out Result)"}.Result
   ) {template="AccumulateLerp::Execute(in TargetValue,in InitialValue,in Blend,in bIntegrateDeltaTime,out Result)"}
   > &RigUnit_SetBoneRotation(Bone=dyn_hairStrandE_03, Weight=1, bPropagateToChildren=True, Rotation=@RigVMStructNode_0_1_1.Result)
   > &RigUnit_SetBoneRotation(
    Bone=dyn_hairStrandA_03
    Weight=1
    bPropagateToChildren=True
    Rotation=&RigVMFunction_AccumulateQuatLerp(
     Blend=0.2
     InitialValue=(X=0,Y=0,Z=0,W=1)
     TargetValue=&RigVMFunction_MathQuaternionSlerp(A=(X=0,Y=0,Z=0,W=1), T=0.65, B=@n4.Transform.Rotation) {template="Interpolate::Execute(in A,in B,in T,out Result)"}.Result
    ) {template="AccumulateLerp::Execute(in TargetValue,in InitialValue,in Blend,in bIntegrateDeltaTime,out Result)"}.Result
   )
   > &RigUnit_SetBoneRotation(
    Bone=dyn_hairStrandB_03
    Weight=1
    bPropagateToChildren=True
    Rotation=&RigVMFunction_AccumulateQuatLerp(
     Blend=0.4
     InitialValue=(X=0,Y=0,Z=0,W=1)
     TargetValue=&RigVMFunction_MathQuaternionSlerp(A=(X=0,Y=0,Z=0,W=1), T=0.5, B=@n4.Transform.Rotation) {template="Interpolate::Execute(in A,in B,in T,out Result)"}.Result
    ) {template="AccumulateLerp::Execute(in TargetValue,in InitialValue,in Blend,in bIntegrateDeltaTime,out Result)"}.Result
   )
   > &RigUnit_SetBoneRotation(
    Bone=dyn_hairStrandG_03
    Weight=1
    bPropagateToChildren=True
    Rotation=&RigVMFunction_AccumulateQuatLerp(
     Blend=0.5
     InitialValue=(X=0,Y=0,Z=0,W=1)
     TargetValue=&RigVMFunction_MathQuaternionSlerp(A=(X=0,Y=0,Z=0,W=1), T=0.725, B=@n4.Transform.Rotation) {template="Interpolate::Execute(in A,in B,in T,out Result)"}.Result
    ) {template="AccumulateLerp::Execute(in TargetValue,in InitialValue,in Blend,in bIntegrateDeltaTime,out Result)"}.Result
   )
   > &RigUnit_SetBoneRotation(
    Bone=dyn_hairStrandH_03
    Weight=1
    bPropagateToChildren=True
    Rotation=&RigVMFunction_AccumulateQuatLerp(
     Blend=0.35
     InitialValue=(X=0,Y=0,Z=0,W=1)
     TargetValue=&RigVMFunction_MathQuaternionSlerp(A=(X=0,Y=0,Z=0,W=1), T=0.8, B=@n4.Transform.Rotation) {template="Interpolate::Execute(in A,in B,in T,out Result)"}.Result
    ) {template="AccumulateLerp::Execute(in TargetValue,in InitialValue,in Blend,in bIntegrateDeltaTime,out Result)"}.Result
   )
   RigVMStructNode_0_1_1_2 = &RigVMFunction_AccumulateQuatLerp(
    Blend=0.2
    InitialValue=(X=0,Y=0,Z=0,W=1)
    TargetValue=&RigVMFunction_MathQuaternionSlerp(A=(X=0,Y=0,Z=0,W=1), T=0.675, B=@n4.Transform.Rotation) {template="Interpolate::Execute(in A,in B,in T,out Result)"}.Result
   ) {template="AccumulateLerp::Execute(in TargetValue,in InitialValue,in Blend,in bIntegrateDeltaTime,out Result)"}
   > &RigUnit_SetBoneRotation(Bone=dyn_hairStrandJ_03, Weight=1, bPropagateToChildren=True, Rotation=@RigVMStructNode_0_1_1_2.Result)
   > &RigUnit_SetBoneRotation(
    Bone=dyn_hairStrandK_03
    Weight=1
    bPropagateToChildren=True
    Rotation=&RigVMFunction_AccumulateQuatLerp(
     Blend=0.2
     InitialValue=(X=0,Y=0,Z=0,W=1)
     TargetValue=&RigVMFunction_MathQuaternionSlerp(A=(X=0,Y=0,Z=0,W=1), T=0.65, B=@n4.Transform.Rotation) {template="Interpolate::Execute(in A,in B,in T,out Result)"}.Result
    ) {template="AccumulateLerp::Execute(in TargetValue,in InitialValue,in Blend,in bIntegrateDeltaTime,out Result)"}.Result
   )
   > &RigUnit_SetBoneRotation(
    Bone=dyn_hairStrandL_03
    Weight=1
    bPropagateToChildren=True
    Rotation=&RigVMFunction_AccumulateQuatLerp(
     Blend=0.4
     InitialValue=(X=0,Y=0,Z=0,W=1)
     TargetValue=&RigVMFunction_MathQuaternionSlerp(A=(X=0,Y=0,Z=0,W=1), T=0.5, B=@n4.Transform.Rotation) {template="Interpolate::Execute(in A,in B,in T,out Result)"}.Result
    ) {template="AccumulateLerp::Execute(in TargetValue,in InitialValue,in Blend,in bIntegrateDeltaTime,out Result)"}.Result
   )
   > &RigUnit_SetBoneRotation(
    Bone=dyn_hairStrandF_03
    Weight=1
    bPropagateToChildren=True
    Rotation=&RigVMFunction_AccumulateQuatLerp(
     Blend=0.4
     InitialValue=(X=0,Y=0,Z=0,W=1)
     TargetValue=&RigVMFunction_MathQuaternionSlerp(A=(X=0,Y=0,Z=0,W=1), T=0.5, B=@n4.Transform.Rotation) {template="Interpolate::Execute(in A,in B,in T,out Result)"}.Result
    ) {template="AccumulateLerp::Execute(in TargetValue,in InitialValue,in Blend,in bIntegrateDeltaTime,out Result)"}.Result
   )
   n5 = &RigUnit_GetBoneTransform(Bone=dyn_hairMain_07, Space=LocalSpace)
   > &RigUnit_SetBoneRotation(
    Bone=dyn_hairStrandB_04
    Weight=1
    bPropagateToChildren=True
    Rotation=&RigVMFunction_AccumulateQuatLerp(
     Blend=0.5
     InitialValue=(X=0,Y=0,Z=0,W=1)
     TargetValue=&RigVMFunction_MathQuaternionSlerp(A=(X=0,Y=0,Z=0,W=1), T=0.725, B=@n5.Transform.Rotation) {template="Interpolate::Execute(in A,in B,in T,out Result)"}.Result
    ) {template="AccumulateLerp::Execute(in TargetValue,in InitialValue,in Blend,in bIntegrateDeltaTime,out Result)"}.Result
   )
   > &RigUnit_SetBoneRotation(
    Bone=dyn_hairStrandA_04
    Weight=1
    bPropagateToChildren=True
    Rotation=&RigVMFunction_AccumulateQuatLerp(
     Blend=0.35
     InitialValue=(X=0,Y=0,Z=0,W=1)
     TargetValue=&RigVMFunction_MathQuaternionSlerp(A=(X=0,Y=0,Z=0,W=1), T=0.8, B=@n5.Transform.Rotation) {template="Interpolate::Execute(in A,in B,in T,out Result)"}.Result
    ) {template="AccumulateLerp::Execute(in TargetValue,in InitialValue,in Blend,in bIntegrateDeltaTime,out Result)"}.Result
   )
   RigVMStructNode_0_1_1_1 = &RigVMFunction_AccumulateQuatLerp(
    Blend=0.2
    InitialValue=(X=0,Y=0,Z=0,W=1)
    TargetValue=&RigVMFunction_MathQuaternionSlerp(A=(X=0,Y=0,Z=0,W=1), T=0.675, B=@n5.Transform.Rotation) {template="Interpolate::Execute(in A,in B,in T,out Result)"}.Result
   ) {template="AccumulateLerp::Execute(in TargetValue,in InitialValue,in Blend,in bIntegrateDeltaTime,out Result)"}
   > &RigUnit_SetBoneRotation(Bone=dyn_hairStrandE_04, Weight=1, bPropagateToChildren=True, Rotation=@RigVMStructNode_0_1_1_1.Result)
   > &RigUnit_SetBoneRotation(
    Bone=dyn_hairStrandC_04
    Weight=1
    bPropagateToChildren=True
    Rotation=&RigVMFunction_AccumulateQuatLerp(
     Blend=0.2
     InitialValue=(X=0,Y=0,Z=0,W=1)
     TargetValue=&RigVMFunction_MathQuaternionSlerp(A=(X=0,Y=0,Z=0,W=1), T=0.65, B=@n5.Transform.Rotation) {template="Interpolate::Execute(in A,in B,in T,out Result)"}.Result
    ) {template="AccumulateLerp::Execute(in TargetValue,in InitialValue,in Blend,in bIntegrateDeltaTime,out Result)"}.Result
   )
   > &RigUnit_SetBoneRotation(
    Bone=dyn_hairStrandD_04
    Weight=1
    bPropagateToChildren=True
    Rotation=&RigVMFunction_AccumulateQuatLerp(
     Blend=0.4
     InitialValue=(X=0,Y=0,Z=0,W=1)
     TargetValue=&RigVMFunction_MathQuaternionSlerp(A=(X=0,Y=0,Z=0,W=1), T=0.5, B=@n5.Transform.Rotation) {template="Interpolate::Execute(in A,in B,in T,out Result)"}.Result
    ) {template="AccumulateLerp::Execute(in TargetValue,in InitialValue,in Blend,in bIntegrateDeltaTime,out Result)"}.Result
   )
   > &RigUnit_SetBoneRotation(
    Bone=dyn_hairStrandG_04
    Weight=1
    bPropagateToChildren=True
    Rotation=&RigVMFunction_AccumulateQuatLerp(
     Blend=0.5
     InitialValue=(X=0,Y=0,Z=0,W=1)
     TargetValue=&RigVMFunction_MathQuaternionSlerp(A=(X=0,Y=0,Z=0,W=1), T=0.725, B=@n5.Transform.Rotation) {template="Interpolate::Execute(in A,in B,in T,out Result)"}.Result
    ) {template="AccumulateLerp::Execute(in TargetValue,in InitialValue,in Blend,in bIntegrateDeltaTime,out Result)"}.Result
   )
   > &RigUnit_SetBoneRotation(
    Bone=dyn_hairStrandH_04
    Weight=1
    bPropagateToChildren=True
    Rotation=&RigVMFunction_AccumulateQuatLerp(
     Blend=0.35
     InitialValue=(X=0,Y=0,Z=0,W=1)
     TargetValue=&RigVMFunction_MathQuaternionSlerp(A=(X=0,Y=0,Z=0,W=1), T=0.8, B=@n5.Transform.Rotation) {template="Interpolate::Execute(in A,in B,in T,out Result)"}.Result
    ) {template="AccumulateLerp::Execute(in TargetValue,in InitialValue,in Blend,in bIntegrateDeltaTime,out Result)"}.Result
   )
   RigVMStructNode_0_1_1_1_1 = &RigVMFunction_AccumulateQuatLerp(
    Blend=0.2
    InitialValue=(X=0,Y=0,Z=0,W=1)
    TargetValue=&RigVMFunction_MathQuaternionSlerp(A=(X=0,Y=0,Z=0,W=1), T=0.675, B=@n5.Transform.Rotation) {template="Interpolate::Execute(in A,in B,in T,out Result)"}.Result
   ) {template="AccumulateLerp::Execute(in TargetValue,in InitialValue,in Blend,in bIntegrateDeltaTime,out Result)"}
   > &RigUnit_SetBoneRotation(Bone=dyn_hairStrandK_04, Weight=1, bPropagateToChildren=True, Rotation=@RigVMStructNode_0_1_1_1_1.Result)
   > &RigUnit_SetBoneRotation(
    Bone=dyn_hairStrandL_04
    Weight=1
    bPropagateToChildren=True
    Rotation=&RigVMFunction_AccumulateQuatLerp(
     Blend=0.2
     InitialValue=(X=0,Y=0,Z=0,W=1)
     TargetValue=&RigVMFunction_MathQuaternionSlerp(A=(X=0,Y=0,Z=0,W=1), T=0.65, B=@n5.Transform.Rotation) {template="Interpolate::Execute(in A,in B,in T,out Result)"}.Result
    ) {template="AccumulateLerp::Execute(in TargetValue,in InitialValue,in Blend,in bIntegrateDeltaTime,out Result)"}.Result
   )
   > &RigUnit_SetBoneRotation(
    Bone=dyn_hairStrandF_04
    Weight=1
    bPropagateToChildren=True
    Rotation=&RigVMFunction_AccumulateQuatLerp(
     Blend=0.2
     InitialValue=(X=0,Y=0,Z=0,W=1)
     TargetValue=&RigVMFunction_MathQuaternionSlerp(A=(X=0,Y=0,Z=0,W=1), T=0.65, B=@n5.Transform.Rotation) {template="Interpolate::Execute(in A,in B,in T,out Result)"}.Result
    ) {template="AccumulateLerp::Execute(in TargetValue,in InitialValue,in Blend,in bIntegrateDeltaTime,out Result)"}.Result
   )
   n6 = &RigUnit_GetBoneTransform(Bone=dyn_hairMain_08, Space=LocalSpace)
   > &RigUnit_SetBoneRotation(
    Bone=dyn_hairStrandE_05
    Weight=1
    bPropagateToChildren=True
    Rotation=&RigVMFunction_AccumulateQuatLerp(
     Blend=0.5
     InitialValue=(X=0,Y=0,Z=0,W=1)
     TargetValue=&RigVMFunction_MathQuaternionSlerp(B=(X=0,Y=0,Z=0,W=1), T=0.725, A=@n6.Transform.Rotation) {template="Interpolate::Execute(in A,in B,in T,out Result)"}.Result
    ) {template="AccumulateLerp::Execute(in TargetValue,in InitialValue,in Blend,in bIntegrateDeltaTime,out Result)"}.Result
   )
   > &RigUnit_SetBoneRotation(
    Bone=dyn_hairStrandD_05
    Weight=1
    bPropagateToChildren=True
    Rotation=&RigVMFunction_AccumulateQuatLerp(Blend=0.45, InitialValue=(X=0,Y=0,Z=0,W=1), TargetValue=(X=0,Y=0,Z=0,W=1)) {template="AccumulateLerp::Execute(in TargetValue,in InitialValue,in Blend,in bIntegrateDeltaTime,out Result)"}.Result
   )
   > &RigUnit_SetBoneRotation(
    Bone=dyn_hairStrandG_05
    Weight=1
    bPropagateToChildren=True
    Rotation=&RigVMFunction_AccumulateQuatLerp(
     Blend=0.5
     InitialValue=(X=0,Y=0,Z=0,W=1)
     TargetValue=&RigVMFunction_MathQuaternionSlerp(A=(X=0,Y=0,Z=0,W=1), T=0.725, B=@n6.Transform.Rotation) {template="Interpolate::Execute(in A,in B,in T,out Result)"}.Result
    ) {template="AccumulateLerp::Execute(in TargetValue,in InitialValue,in Blend,in bIntegrateDeltaTime,out Result)"}.Result
   )
   > &RigUnit_SetBoneRotation(
    Bone=dyn_hairStrandH_05
    Weight=1
    bPropagateToChildren=True
    Rotation=&RigVMFunction_AccumulateQuatLerp(
     Blend=0.35
     InitialValue=(X=0,Y=0,Z=0,W=1)
     TargetValue=&RigVMFunction_MathQuaternionSlerp(A=(X=0,Y=0,Z=0,W=1), T=0.8, B=@n6.Transform.Rotation) {template="Interpolate::Execute(in A,in B,in T,out Result)"}.Result
    ) {template="AccumulateLerp::Execute(in TargetValue,in InitialValue,in Blend,in bIntegrateDeltaTime,out Result)"}.Result
   )
   RigVMStructNode_0_1_1_1_1_1 = &RigVMFunction_AccumulateQuatLerp(
    Blend=0.2
    InitialValue=(X=0,Y=0,Z=0,W=1)
    TargetValue=&RigVMFunction_MathQuaternionSlerp(A=(X=0,Y=0,Z=0,W=1), T=0.675, B=@n6.Transform.Rotation) {template="Interpolate::Execute(in A,in B,in T,out Result)"}.Result
   ) {template="AccumulateLerp::Execute(in TargetValue,in InitialValue,in Blend,in bIntegrateDeltaTime,out Result)"}
   > &RigUnit_SetBoneRotation(Bone=dyn_hairStrandL_05, Weight=1, bPropagateToChildren=True, Rotation=@RigVMStructNode_0_1_1_1_1_1.Result)
   n7 = &RigVMFunction_MathQuaternionToEuler(RotationOrder=ZYX, Value=&RigUnit_GetRelativeBoneTransform(Bone=head, Space=spine_05).Transform.Rotation)
   n8 = RigVMVariableNode() {variable=backblingFloat}
   > &RigUnit_ModifyBoneTransforms(
    BoneToModify=((Bone="dyn_hairMain_01_DRV",Transform=(Rotation=(X=0,Y=0,Z=0,W=1),Translation=(X=-2,Y=0,Z=0),Scale3D=(X=1,Y=1,Z=1))))
    Mode=AdditiveLocal
    WeightMaximum=1
    Weight=&RigVMFunction_MathFloatMul(A=&RigVMFunction_MathFloatRemap(SourceMaximum=-42, bClamp=true, Value=@n7.Result.Z).Result, B=@n8.Value) {template="Multiply::Execute(in A,in B,out Result)"}.Result
    BoneToModify.0.Transform.Rotation=&RigVMFunction_MathQuaternionFromEuler(Euler=(X=0,Y=10,Z=-10)).Result
   )
   > &RigUnit_ModifyBoneTransforms(
    BoneToModify=((Bone="dyn_hairMain_01_DRV",Transform=(Rotation=(X=0,Y=0,Z=0,W=1),Translation=(X=0,Y=0,Z=0),Scale3D=(X=1,Y=1,Z=1))))
    Mode=AdditiveLocal
    WeightMaximum=1
    Weight=&RigVMFunction_MathFloatMul(A=&RigVMFunction_MathFloatRemap(SourceMaximum=-25, SourceMinimum=-5, bClamp=true, Value=@n7.Result.Y).Result, B=@n8.Value) {template="Multiply::Execute(in A,in B,out Result)"}.Result
    BoneToModify.0.Transform.Rotation=&RigVMFunction_MathQuaternionFromEuler(Euler=(X=-15,Y=15,Z=0)).Result
   )
   n9 = /Script/RigVM.RigVMFunction_MathFloatEquals(B=1, A=@n8.Value) {is=struct RigVMUnitNode}
   > &RigUnit_ModifyBoneTransforms(
    BoneToModify=((Bone="dyn_hairMain_01_DRV",Transform=(Rotation=(X=0,Y=0,Z=0,W=1),Translation=(X=0,Y=0,Z=0),Scale3D=(X=1,Y=1,Z=1))))
    Mode=AdditiveLocal
    WeightMaximum=1
    Weight=&RigVMFunction_MathFloatMul(
     A=&RigVMFunction_MathFloatRemap(SourceMaximum=26, SourceMinimum=10, TargetMinimum=0, bClamp=true, Value=@n7.Result.Z).Result
     B=/Script/RigVM.RigVMFunction_MathFloatSelectBool(IfFalse=1, IfTrue=0.5, Condition=@n9.Result) {is=struct RigVMUnitNode}.Result
    ) {template="Multiply::Execute(in A,in B,out Result)"}.Result
    BoneToModify.0.Transform.Rotation=&RigVMFunction_MathQuaternionFromEuler(Euler=(X=0,Y=0,Z=-20)).Result
   )
   &RigUnit_ModifyBoneTransforms(
    BoneToModify=((Bone="dyn_hairMain_01_DRV_Base",Transform=(Rotation=(X=0,Y=0,Z=0,W=1),Translation=(X=0,Y=0,Z=0),Scale3D=(X=0.25,Y=1,Z=1))))
    ExecutePin=()
    Mode=AdditiveLocal
    Weight=1
    WeightMaximum=1
   )
   &RigUnit_ModifyBoneTransforms(
    BoneToModify=((Bone="dyn_hairMain_02",Transform=(Rotation=(X=0,Y=0,Z=0,W=1),Translation=(X=-20,Y=0,Z=0),Scale3D=(X=1,Y=1,Z=1))))
    ExecutePin=()
    Mode=AdditiveLocal
    Weight=1
    WeightMaximum=1
    BoneToModify.0.Transform.Rotation=&RigVMFunction_MathQuaternionFromEuler(Euler=(X=0,Y=0,Z=10)).Result
   )
  ]}
]
hierarchy [
 root {loc=<0 0 0>, rot=<0 0 0 1>}
 pelvis {parent=root, loc=<0 -0.5301 98.0571>, rot=<0.0127 0.707 -0.0127 -0.707>}
 spine_01 {parent=pelvis, loc=<6.1208 -0.9328 0>, rot=<0 0 0.034 -0.9994>}
 spine_02 {parent=spine_01, loc=<6.5996 -0.2095 0>, rot=<0 0 0.0238 -0.9997>}
 spine_03 {parent=spine_02, loc=<6.4914 -0.1237 0>, rot=<0 0 0.0149 -0.9999>}
 spine_04 {parent=spine_03, loc=<6.9598 0.0794 0>, rot=<0 0 -0.0132 -0.9999>}
 spine_05 {parent=spine_04, loc=<6.9987 0 0>, rot=<0 0 -0.0015 -1>}
 neck_01 {parent=spine_05, loc=<11.3138 0.1529 0>, rot=<0 0 -0.2003 -0.9797>}
 neck_02 {parent=neck_01, loc=<4.5627 0 0>, rot=<0 0 0.0504 -0.9987>}
 head {parent=neck_02, loc=<5.3914 0 0>, rot=<0 0 0.115 -0.9934>}
 dyn_hairBangC_01_r {parent=head, loc=<7.4853 4.2672 6.8193>, rot=<0.1926 -0.5311 0.8238 -0.0468>}
 dyn_hairBangC_02_r {parent=dyn_hairBangC_01_r, loc=<6.0316 0 0>, rot=<0.2055 0.0474 -0.1368 -0.9679>}
 dyn_hairBangC_03_r {parent=dyn_hairBangC_02_r, loc=<9.254 0 0>, rot=<0.2656 0.1146 0.0705 -0.9547>}
 dyn_hairBangC_04_r {parent=dyn_hairBangC_03_r, loc=<5.6185 0 0>, rot=<-0.0647 0.1077 0.1986 -0.972>}
 dyn_hairBangB_01_r {parent=head, loc=<6.445 7.1872 6.6975>, rot=<0.1892 -0.5888 0.7835 -0.0602>}
 dyn_hairBangB_02_r {parent=dyn_hairBangB_01_r, loc=<5.8495 0 0>, rot=<-0.0116 0.0547 -0.0447 -0.9974>}
 dyn_hairBangB_03_r {parent=dyn_hairBangB_02_r, loc=<4.733 0 0>, rot=<0.2991 0.0891 0.0371 -0.9493>}
 dyn_hairBangB_04_r {parent=dyn_hairBangB_03_r, loc=<5.4031 0 0>, rot=<-0.021 0.0827 0.0978 -0.9915>}
 dyn_hairBangA_01_r {parent=head, loc=<9.1875 9.8076 2.3857>, rot=<0.2276 -0.3583 0.9006 0.0934>}
 dyn_hairBangA_02_r {parent=dyn_hairBangA_01_r, loc=<5.1114 0 0>, rot=<0.0582 0.0909 -0.1685 -0.9798>}
 dyn_hairBangA_03_r {parent=dyn_hairBangA_02_r, loc=<4.65 0 0>, rot=<0.058 0.1299 -0.0199 -0.9896>}
 dyn_hairBangA_04_r {parent=dyn_hairBangA_03_r, loc=<4.2298 0 0>, rot=<-0.0212 0.0942 0.0413 -0.9945>}
 dyn_hairBangA_01_l {parent=head, loc=<8.4849 9.3474 -3.0883>, rot=<-0.1225 0.4759 0.8452 0.2101>}
 dyn_hairBangA_02_l {parent=dyn_hairBangA_01_l, loc=<3.9583 0 0>, rot=<-0.0801 -0.0273 -0.1658 -0.9825>}
 dyn_hairBangA_03_l {parent=dyn_hairBangA_02_l, loc=<3.5995 0 0>, rot=<0.0891 -0.0141 -0.1254 -0.988>}
 dyn_hairBangA_04_l {parent=dyn_hairBangA_03_l, loc=<2.4221 0 0>, rot=<-0.0181 -0.0375 -0.1149 -0.9925>}
 dyn_hairBangB_01_l {parent=head, loc=<7.3687 9.378 -4.9537>, rot=<-0.1133 0.3924 0.9113 0.0524>}
 dyn_hairBangB_02_l {parent=dyn_hairBangB_01_l, loc=<4.393 0 0>, rot=<0.0495 0.0137 -0.157 -0.9863>}
 dyn_hairBangB_03_l {parent=dyn_hairBangB_02_l, loc=<3.6291 0 0>, rot=<-0.1346 -0.0326 -0.0869 -0.9865>}
 dyn_hairBangB_04_l {parent=dyn_hairBangB_03_l, loc=<5.107 0 0>, rot=<-0.0096 -0.0738 0.0105 -0.9972>}
 dyn_hairBangC_01_l {parent=head, loc=<4.9731 7.5502 -6.2138>, rot=<0.1143 -0.7286 -0.6672 0.1048>}
 dyn_hairBangC_02_l {parent=dyn_hairBangC_01_l, loc=<5.5677 0 0>, rot=<-0.0082 -0.1313 0.0076 -0.9913>}
 dyn_hairBangC_03_l {parent=dyn_hairBangC_02_l, loc=<6.7245 0 0>, rot=<0.0432 -0.0997 0.0431 -0.9931>}
 dyn_hairBangD_01_l {parent=head, loc=<8.4957 5.959 -6.9092>, rot=<-0.1927 0.6417 0.7413 -0.0397>}
 dyn_hairBangD_02_l {parent=dyn_hairBangD_01_l, loc=<6.5057 0 0>, rot=<0.006 -0.0896 -0.1255 -0.988>}
 dyn_hairBangD_03_l {parent=dyn_hairBangD_02_l, loc=<6.7867 0 0>, rot=<-0.0695 -0.0586 -0.0508 -0.9946>}
 dyn_hairBangD_04_l {parent=dyn_hairBangD_03_l, loc=<5.141 0 0>, rot=<0.0653 -0.0952 0.0508 -0.992>}
 dyn_hairBangD_05_l {parent=dyn_hairBangD_04_l, loc=<3.0263 0 0>, rot=<-0.0426 -0.0026 -0.0472 -0.998>}
 dyn_hairMain_01_DRV_Base {parent=head, loc=<3.1773 -6.0503 0.4307>, rot=<0.1753 -0.9841 -0.0051 0.0298>}
 dyn_hairMain_01_DRV {parent=dyn_hairMain_01_DRV_Base, loc=<0 0 0>, rot=<0 0 0 -1>}
 dyn_hairMain_01 {parent=dyn_hairMain_01_DRV, loc=<0 0 0>, rot=<0 0 0 -1>}
 dyn_hairMain_02 {parent=dyn_hairMain_01, loc=<8.3677 0 0>, rot=<-0.0031 0.0102 -0.0474 -0.9988>}
 dyn_hairMain_03 {parent=dyn_hairMain_02, loc=<9.1931 0 0>, rot=<-0.0044 0.02 -0.0326 -0.9993>}
 dyn_hairMain_04 {parent=dyn_hairMain_03, loc=<13.4151 0 0>, rot=<-0.0014 0.0069 0.0101 -0.9999>}
 dyn_hairMain_05 {parent=dyn_hairMain_04, loc=<14.2521 0 0>, rot=<-0.0001 0.0007 -0.0007 -1>}
 dyn_hairMain_06 {parent=dyn_hairMain_05, loc=<14.6586 0 0>, rot=<-0.0018 0.0103 -0.0256 -0.9996>}
 dyn_hairMain_07 {parent=dyn_hairMain_06, loc=<13.8729 0 0>, rot=<-0.0023 0.016 -0.0067 -0.9998>}
 dyn_hairMain_08 {parent=dyn_hairMain_07, loc=<10.242 0 0>, rot=<0.0033 -0.0245 -0.0043 -0.9997>}
 dyn_hairMain_09 {parent=dyn_hairMain_08, loc=<7.449 0 0>, rot=<0.0011 -0.0094 -0.0101 -0.9999>}
 dyn_hairStrandK_01 {parent=dyn_hairMain_06, loc=<1.4331 1.3229 -2.985>, rot=<0.1794 0.0084 -0.0037 -0.9837>}
 dyn_hairStrandK_02 {parent=dyn_hairStrandK_01, loc=<8.6509 0 0>, rot=<-0.0376 -0.0059 0.0047 -0.9993>}
 dyn_hairStrandK_03 {parent=dyn_hairStrandK_02, loc=<7.1405 0 0>, rot=<0.0457 0.0061 -0.0279 -0.9985>}
 dyn_hairStrandK_04 {parent=dyn_hairStrandK_03, loc=<6.6018 0 0>, rot=<-0.1783 -0.0183 -0.0019 -0.9838>}
 dyn_hairStrandH_01 {parent=dyn_hairMain_05, loc=<6.2154 0.8149 1.1505>, rot=<-0.0963 -0.028 -0.0043 -0.9949>}
 dyn_hairStrandH_02 {parent=dyn_hairStrandH_01, loc=<14.074 0 0>, rot=<-0.1146 -0.019 -0.0283 -0.9928>}
 dyn_hairStrandH_03 {parent=dyn_hairStrandH_02, loc=<8.9618 0 0>, rot=<0.1277 0.017 -0.003 -0.9917>}
 dyn_hairStrandH_04 {parent=dyn_hairStrandH_03, loc=<7.4206 0 0>, rot=<0.0285 0.0035 -0.0087 -0.9995>}
 dyn_hairStrandH_05 {parent=dyn_hairStrandH_04, loc=<4.4241 0 0>, rot=<0.1149 0.0111 -0.0154 -0.9932>}
 dyn_hairStrandJ_01 {parent=dyn_hairMain_05, loc=<13.9085 1.0851 -3.9422>, rot=<0.182 0.0276 -0.008 -0.9829>}
 dyn_hairStrandJ_02 {parent=dyn_hairStrandJ_01, loc=<9.3752 0 0>, rot=<0.0466 0.0084 -0.0126 -0.9988>}
 dyn_hairStrandJ_03 {parent=dyn_hairStrandJ_02, loc=<11.4652 0 0>, rot=<-0.2301 -0.0371 -0.0049 -0.9725>}
 dyn_hairStrandL_01 {parent=dyn_hairMain_05, loc=<5.0248 -2.2082 3.4718>, rot=<0.002 -0.0103 -0.0106 -0.9999>}
 dyn_hairStrandL_02 {parent=dyn_hairStrandL_01, loc=<9.5796 0 0>, rot=<-0.0009 0.0051 -0.0065 -1>}
 dyn_hairStrandL_03 {parent=dyn_hairStrandL_02, loc=<9.8336 0 0>, rot=<-0.0012 0.0075 -0.0089 -0.9999>}
 dyn_hairStrandL_04 {parent=dyn_hairStrandL_03, loc=<10.2496 0 0>, rot=<-0.0026 0.0158 0.01 -0.9998>}
 dyn_hairStrandL_05 {parent=dyn_hairStrandL_04, loc=<5.1556 0 0>, rot=<-0.0587 0.6854 -0.0619 -0.7232>}
 dyn_hairStrandC_01 {parent=dyn_hairMain_04, loc=<5.2861 -2.9099 0.8954>, rot=<0.0023 -0.0106 0.0056 -0.9999>}
 dyn_hairStrandC_02 {parent=dyn_hairStrandC_01, loc=<14.6532 0 0>, rot=<-0.0013 0.0066 -0.0189 -0.9998>}
 dyn_hairStrandC_03 {parent=dyn_hairStrandC_02, loc=<9.5554 0 0>, rot=<0.002 -0.0125 -0.0176 -0.9998>}
 dyn_hairStrandC_04 {parent=dyn_hairStrandC_03, loc=<9.5546 0 0>, rot=<-0.0004 0.0029 -0.0002 -1>}
 dyn_hairStrandE_01 {parent=dyn_hairMain_04, loc=<0.7151 0.554 -5.0131>, rot=<-0.0043 0.0198 0.011 -0.9997>}
 dyn_hairStrandE_02 {parent=dyn_hairStrandE_01, loc=<14.6496 0 0>, rot=<0.0007 -0.0027 0.0103 -0.9999>}
 dyn_hairStrandE_03 {parent=dyn_hairStrandE_02, loc=<9.8326 0 0>, rot=<0.0002 -0.0008 -0.0197 -0.9998>}
 dyn_hairStrandE_04 {parent=dyn_hairStrandE_03, loc=<9.845 0 0>, rot=<0.0026 -0.0132 -0.0106 -0.9999>}
 dyn_hairStrandE_05 {parent=dyn_hairStrandE_04, loc=<8.6525 0 0>, rot=<0.0074 -0.0341 0.0278 -0.999>}
 dyn_hairStrandF_01 {parent=dyn_hairMain_04, loc=<14.3949 0.3399 5.6939>, rot=<-0.0388 -0.0157 -0.0091 -0.9991>}
 dyn_hairStrandF_02 {parent=dyn_hairStrandF_01, loc=<9.6519 0 0>, rot=<-0.0027 -0.0005 0.002 -1>}
 dyn_hairStrandF_03 {parent=dyn_hairStrandF_02, loc=<9.9005 0 0>, rot=<-0.0362 -0.0066 -0.0082 -0.9993>}
 dyn_hairStrandF_04 {parent=dyn_hairStrandF_03, loc=<9.4484 0 0>, rot=<0.0055 0.0008 -0.0232 -0.9997>}
 dyn_hairStrandG_01 {parent=dyn_hairMain_04, loc=<11.5905 -1.1448 2.9493>, rot=<-0.0366 -0.0151 -0.0143 -0.9991>}
 dyn_hairStrandG_02 {parent=dyn_hairStrandG_01, loc=<5.0765 0 0>, rot=<0.022 0.0038 -0.0032 -0.9997>}
 dyn_hairStrandG_03 {parent=dyn_hairStrandG_02, loc=<10.1237 0 0>, rot=<0.041 0.0066 -0.0086 -0.9991>}
 dyn_hairStrandG_04 {parent=dyn_hairStrandG_03, loc=<10.7685 0 0>, rot=<-0.0039 -0.0006 -0.0093 -0.9999>}
 dyn_hairStrandG_05 {parent=dyn_hairStrandG_04, loc=<10.6391 0 0>, rot=<-0.0446 -0.005 -0.0215 -0.9988>}
 dyn_hairStrandA_01 {parent=dyn_hairMain_03, loc=<4.5931 0.7193 8.0753>, rot=<0.0046 -0.0255 -0.0055 -0.9996>}
 dyn_hairStrandA_02 {parent=dyn_hairStrandA_01, loc=<14.0319 0 0>, rot=<-0.0029 0.0147 0.0189 -0.9997>}
 dyn_hairStrandA_03 {parent=dyn_hairStrandA_02, loc=<13.7769 0 0>, rot=<-0.0011 0.0052 0.0061 -1>}
 dyn_hairStrandA_04 {parent=dyn_hairStrandA_03, loc=<9.6579 0 0>, rot=<-0.0024 0.0101 0.0075 -0.9999>}
 dyn_hairStrandB_01 {parent=dyn_hairMain_03, loc=<13.6266 -1.6777 5.4125>, rot=<0.0017 -0.0086 0.0123 -0.9999>}
 dyn_hairStrandB_02 {parent=dyn_hairStrandB_01, loc=<14.0633 0 0>, rot=<-0.0024 0.0121 -0.0105 -0.9999>}
 dyn_hairStrandB_03 {parent=dyn_hairStrandB_02, loc=<14.57 0 0>, rot=<-0.0013 0.0077 -0.0224 -0.9997>}
 dyn_hairStrandB_04 {parent=dyn_hairStrandB_03, loc=<9.4928 0 0>, rot=<-0.0012 0.0083 0.0001 -1>}
 dyn_hairStrandD_01 {parent=dyn_hairMain_03, loc=<8.7729 -0.5515 -4.0344>, rot=<-0.0037 0.0182 0.0141 -0.9997>}
 dyn_hairStrandD_02 {parent=dyn_hairStrandD_01, loc=<13.7876 0 0>, rot=<-0.0008 0.0039 -0.0004 -1>}
 dyn_hairStrandD_03 {parent=dyn_hairStrandD_02, loc=<14.3422 0 0>, rot=<-0.005 0.0243 -0.0094 -0.9996>}
 dyn_hairStrandD_04 {parent=dyn_hairStrandD_03, loc=<9.8901 0 0>, rot=<0.0006 -0.0034 -0.0031 -1>}
 dyn_hairStrandD_05 {parent=dyn_hairStrandD_04, loc=<9.097 0 0>, rot=<0.0032 -0.0174 -0.0066 -0.9998>}
]
variables [backblingFloat {type=double, default=0}]
