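# Post-process style cosmetic AnimBP for the female Angler body part, adding hair secondary motion to the pose copied from the attached parent through Control Rig.
# With rigid bodies enabled, backbling selects the physics asset, then bangs physics, a second Control Rig and backbling hair shrinking follow.
# Otherwise, head-driven bone offsets feed a hair Trail controller.
# Speed, movement direction and backbling drive weights and relaxation; skydiving start resets dynamics.
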AnimBlueprint /BRCosmetics/Characters/Player/Female/Medium/Bodies/F_MED_Angler/Meshes/Parts/F_MED_Angler_AnimBP
target_skeleton=&/F_MED_Angler_Skeleton
graphs [
 AnimGraph {nodes [
   Output Pose(Result=Blend Poses by bool(BlendPose_0=Use cached pose 'simCachePose'(), BlendPose_1=Use cached pose 'NoSimCachePose'(), bActiveValue=Get bRigidBodyEnabled()) {blendType=Linear}) {layerGroup=None}
   n1 = Get backblingFloat() -> (backblingFloat:double)
   n2 = Control Rig!(Source=Copy Pose From Mesh() {bUseAttachedParent=true}, Alpha=Get ctlrRigAlpha_A() -> (ctlrRigAlpha_A:double), backblingFloat:double=@n1) -> (Pose) {
    alphaBoolBlend {blendOption=Linear}
    controlRigAssetReference {blueprintRigClass=/BRCosmetics/Characters/Player/Female/Medium/Bodies/F_MED_Angler/Meshes/Parts/F_MED_Angler_CtrlRig/F_MED_Angler_CtrlRig.F_MED_Angler_CtrlRig_C}
   }
   Save cached pose 'preRigidCache'(Pose=@n2) {cacheName=preRigidCache}
   n3 = Get head_fwdBwd_rot() -> (head_fwdBwd_rot:double)
   n4 = "Transform (Modify) Bone - Bone: dyn_hairMain_01_DRV"(
    ComponentPose=Local To Component(LocalPose=Use cached pose 'preRigidCache'())
    Alpha=@n3
    Translation="0,-10,0"
    Rotation="0,-50,0"
    Scale="1,1,1"
   ) {
    rotationMode=BMM_Additive
    rotationSpace=BCS_BoneSpace
    scaleMode=BMM_Additive
    scaleSpace=BCS_BoneSpace
    translationMode=BMM_Additive
    translationSpace=BCS_BoneSpace
    alphaBoolBlend {blendOption=Linear}
    alphaScaleBiasClamp {bMapRange=true, inRange {max=0, min=-30}, outRange {max=0, min=1}}
    boneToModify {boneName=dyn_hairMain_01_DRV}
    rotation {yaw=-50}
    translation {y=-10}
   }
   n5 = "Transform (Modify) Bone - Bone: dyn_hairMain_02"(ComponentPose=@n4, Alpha=@n3, Translation="0,5,0", Rotation="0,-5,0", Scale="1,1,1") {
    rotationMode=BMM_Additive
    rotationSpace=BCS_BoneSpace
    scaleMode=BMM_Additive
    scaleSpace=BCS_BoneSpace
    translationMode=BMM_Additive
    translationSpace=BCS_BoneSpace
    alphaBoolBlend {blendOption=Linear}
    alphaScaleBiasClamp {bMapRange=true, inRange {max=0, min=-30}, outRange {max=0, min=1}}
    boneToModify {boneName=dyn_hairMain_02}
    rotation {yaw=-5}
    translation {y=5}
   }
   n6 = Get head_fwdBwd_rot()
   n7 = "Transform (Modify) Bone - Bone: dyn_hairMain_01_DRV"(ComponentPose=@n5, Alpha=@n6, Translation="0,0,0", Rotation="0,0,0", Scale="1,1,1") {
    rotationMode=BMM_Additive
    rotationSpace=BCS_BoneSpace
    scaleMode=BMM_Additive
    scaleSpace=BCS_BoneSpace
    translationMode=BMM_Additive
    translationSpace=BCS_BoneSpace
    alphaBoolBlend {blendOption=Linear}
    alphaScaleBiasClamp {bMapRange=true, inRange {max=30}}
    boneToModify {boneName=dyn_hairMain_01_DRV}
   }
   n8 = "Transform (Modify) Bone - Bone: dyn_hairMain_03"(ComponentPose=@n7, Alpha=@n6, Translation="0,5,0", Rotation="0,40,0", Scale="1,1,1") {
    rotationMode=BMM_Additive
    rotationSpace=BCS_BoneSpace
    scaleMode=BMM_Additive
    scaleSpace=BCS_BoneSpace
    translationMode=BMM_Additive
    translationSpace=BCS_BoneSpace
    alphaBoolBlend {blendOption=Linear}
    alphaScaleBiasClamp {bMapRange=true, inRange {max=30}}
    boneToModify {boneName=dyn_hairMain_03}
    rotation {yaw=40}
    translation {y=5}
   }
   n9 = Trail controller - Bone: dyn_hairMain_08!(ComponentPose=@n8, Alpha=1, RelaxationSpeedScale:float=Get RelaxSpeed() -> (RelaxSpeed:double)) -> (Pose) {
    bLimitRotation=true
    chainLength=8
    maxDeltaTime=0.0333
    alphaScaleBiasClamp {bInterpResult=true}
    rotationLimits [
     {limitMax=<180 180 180>}
     {limitMax=<180 180 180>}
     {limitMax=<180 180 180>}
     {limitMax=<180 180 180>}
     {limitMax=<180 180 180>}
     {limitMax=<180 180 180>}
     {limitMax=<180 180 180>}
     {limitMax=<180 180 180>}
    ]
    rotationOffsets [<0 0 0>, <0 0 0>, <0 0 0>, <0 0 0>, <0 0 0>, <0 0 0>, <0 0 0>, <0 0 0>]
    trailBone {boneName=dyn_hairMain_08}
    trailRelaxationSpeed {editorCurveData {keys [<at 0 10 RCIM_Linear>,<at 0.5107 8.1304 RCIM_Linear>,<at 1 5 RCIM_Linear>]}}
   }
   Save cached pose 'NoSimCachePose'(Pose=Component To Local(ComponentPose=@n9)) {cacheName=NoSimCachePose}
   n10 = RigidBody(ComponentPose=Local To Component(LocalPose=Use cached pose 'preRigidCache'()), Alpha=Get rigidAlpha_A() -> (rigidAlpha_A:double)) {
    bClampLinearTranslationLimitToRefPose=true
    bOverrideWorldGravity=true
    overridePhysicsAsset=&/F_MED_Angler_Physics
    alphaBoolBlend {blendOption=Linear}
    baseBoneRef {boneName=root}
    componentAppliedLinearAccClamp=<3000 3000 4000>
    componentLinearAccScale=<0.5 0.5 0.4>
    componentLinearVelScale=<0.2 0.2 0.2>
    overrideWorldGravity {z=-100}
   }
   n11 = Component To Local(ComponentPose=@n10)
   n12 = RigidBody(ComponentPose=Local To Component(LocalPose=Use cached pose 'preRigidCache'()), Alpha=Get rigidAlpha_A()) {
    bClampLinearTranslationLimitToRefPose=true
    bOverrideWorldGravity=true
    overridePhysicsAsset=&/F_MED_Angler_BackpackOn_Physics
    alphaBoolBlend {blendOption=Linear}
    baseBoneRef {boneName=root}
    componentAppliedLinearAccClamp=<3000 3000 4000>
    componentLinearAccScale=<0.5 0.5 0.45>
    componentLinearVelScale=<0.2 0.2 0.2>
    overrideWorldGravity {z=-100}
   }
   n13 = Component To Local(ComponentPose=@n12)
   n14 = Get backblingOn() -> (backblingOn:bool)
   n15 = RigidBody(
    ComponentPose=Local To Component(LocalPose=Blend Poses by bool(BlendPose_0=@n13, BlendPose_1=@n11, bActiveValue=@n14) {blendType=Linear})
    Alpha=Get rigidAlpha_B() -> (rigidAlpha_B:double)
   ) {
    bClampLinearTranslationLimitToRefPose=true
    bOverrideWorldGravity=true
    overridePhysicsAsset=&/F_MED_Angler_BangsPhysics
    simulationSpace=BaseBoneSpace
    alphaBoolBlend {blendOption=Linear}
    baseBoneRef {boneName=root}
    componentAppliedLinearAccClamp=<2000 2000 2000>
    componentLinearAccScale=<0.15 0.15 0.15>
    componentLinearVelScale=<0.3 0.3 3>
    overrideWorldGravity {z=-450}
   }
   n16 = Control Rig(Source=Component To Local(ComponentPose=@n15), Alpha=Get ctrlRigAlpha_B() -> (ctrlRigAlpha_B:double), backblingFloat=@n1) {
    alphaBoolBlend {blendOption=Linear}
    controlRigAssetReference {blueprintRigClass=&F_MED_Angler_CtrlRig_C}
   }
   n17 = Get backblingOn()
   n18 = "Transform (Modify) Bone - Bone: dyn_hairMain_01"!(ComponentPose=Local To Component(LocalPose=@n16), bAlphaBoolEnabled:bool=@n17, Translation="0,0,-2", Rotation="0,0,0", Scale="0.6,1,1") -> (Pose) {
    alphaInputType=Bool
    scaleMode=BMM_Replace
    translationMode=BMM_Additive
    alphaBoolBlend {blendOption=Linear}
    boneToModify {boneName=dyn_hairMain_01}
    scale {x=0.6}
    translation {z=-2}
   }
   n19 = "Transform (Modify) Bone - Bone: dyn_hairMain_02"!(ComponentPose=@n18, bAlphaBoolEnabled:bool=@n17, Translation="0,-2,-2", Rotation="0,0,0", Scale="0.4,1,1") -> (Pose) {
    alphaInputType=Bool
    scaleMode=BMM_Replace
    translationMode=BMM_Additive
    alphaBoolBlend {blendOption=Linear}
    boneToModify {boneName=dyn_hairMain_02}
    scale {x=0.4}
    translation {y=-2, z=-2}
   }
   n20 = "Transform (Modify) Bone - Bone: dyn_hairMain_03"!(ComponentPose=@n19, bAlphaBoolEnabled:bool=@n17, Translation="0,-2,-2", Rotation="0,0,0", Scale="0.2,1,1") -> (Pose) {
    alphaInputType=Bool
    scaleMode=BMM_Replace
    translationMode=BMM_Additive
    alphaBoolBlend {blendOption=Linear}
    boneToModify {boneName=dyn_hairMain_03}
    scale {x=0.2}
    translation {y=-2, z=-2}
   }
   Save cached pose 'simCachePose'(Pose=Component To Local(ComponentPose=@n20)) {cacheName=simCachePose}
  ]}
 EventGraph {nodes [
   Event BlueprintUpdateAnimation()
   > n1 = Sequence!(execute) -> (then_0, then_1, then_2:Exec, then_3:Exec)
   n2 = CheckForBackbling()
   Set backblingOn(execute=@n1.then_3, backblingOn:bool=@n2) -> (Output_Get:bool)
   > Set backblingFloat(backblingFloat:double=SelectFloat(A=0.5, B=1, bPickA=@n2)) -> (Output_Get:double)
   n3 = Get PawnSpeed2D()
   Set ctlrRigAlpha_A(
    execute=@n1.then_0
    ctlrRigAlpha_A:double=float * float(A=MapRangeClamped(Value=@n3, InRangeA=190, InRangeB=550, OutRangeA=0.3, OutRangeB=0.1), B=Get backblingFloat())
   ) -> (Output_Get:double)
   > Set ctrlRigAlpha_B(ctrlRigAlpha_B:double=float * float(A=MapRangeClamped(Value=@n3, InRangeA=190, InRangeB=550, OutRangeA=0.8, OutRangeB=0.1), B=Get backblingFloat())) -> (Output_Get:double)
   > Set rigidAlpha_A(rigidAlpha_A:double=MapRangeClamped(Value=@n3, InRangeA=190, InRangeB=550, OutRangeA=0.8)) -> (Output_Get:double)
   > Set rigidAlpha_B(rigidAlpha_B:double=MapRangeClamped(Value=@n3, InRangeA=190, InRangeB=550, OutRangeA=0.8)) -> (Output_Get:double)
   n4 = GetDeltaTransformFromRefPose(self=GetOwnerSkeletalMeshComponent(), BoneName=head, BaseName=neck_01)
   Set head_fwdBwd_rot(execute=@n1.then_2, head_fwdBwd_rot:double=BreakRotator(InRot=BreakTransform(InTransform=@n4).Rotation).Yaw) -> (Output_Get:double)
   n5 = GetOwnerDirection!(self:Custom Character Part Anim Instance Object Reference) -> (ReturnValue:float)
   Set RelaxSpeed(
    execute=@n1.then_1
    RelaxSpeed:double=MapRangeClamped(Value="Absolute (Float)"(A=@n5), InRangeA=85, InRangeB=110, OutRangeA=1.5, OutRangeB=4)
   ) -> (Output_Get:double)
   > n6 = Branch(Condition=Get bIsSkyDiving())
   > Set TrailAlpha(TrailAlpha:double=0.5) -> (Output_Get:double)
   n7 = Branch(execute=@n6.else, Condition=OR Boolean(A=Get bIsJumping(), B=Get bIsFalling()))
   > Set TrailAlpha(TrailAlpha=0.75)
   Set TrailAlpha(execute=@n7.else, TrailAlpha=1)
   Event OnBeginSkydiving()
   > Delay(Duration=0.1)
   > ResetDynamics(InTeleportType=TeleportPhysics)
   > ResetDynamics(InTeleportType=ResetPhysics)
   Event OnCharacterCustomizationLoaded()
   Do Once()
  ]}
]
variables [
 backblingFloat {type=number}
 backblingOn {type=boolean}
 ctlrRigAlpha_A {type=number}
 ctrlRigAlpha_B {type=number}
 head_fwdBwd_rot {type=number}
 RelaxSpeed {type=number}
 rigidAlpha_A {type=number}
 rigidAlpha_B {type=number}
 TrailAlpha {type=number}
]
